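# Material function for the WeaveHarbor backpack cosmetic that blends through three text textures based on a Phase input.
# It samples Text01, Text02 and Text03 texture parameters with shared UV and MipBias inputs, using shared clamp samplers.
# Two chained lerps, driven by saturated Phase and Phase minus one, crossfade them in sequence across a 0 to 2 range.
# The result outputs as RGB and as separate R, G, B and A channels.
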MaterialFunction /BRCosmetics/Accessories/FORT_Backpacks/Backpack_M_MED_WeaveHarbor/Materials/MF_Backpack_WeaveHarbor_TextureSwap
graph {nodes [
  n1 = FunctionInput(Preview=TextureCoordinate()) {bUsePreviewValueAsDefault=true, inputName=UV, inputType=FunctionInput_Vector2}
  n2 = FunctionInput() {bUsePreviewValueAsDefault=true, inputName=MipBias, inputType=FunctionInput_Scalar}
  n3 = TextureSampleParameter2D!(UVs=@n1, Bias=@n2, Apply View MipBias) {
   group=WeaveHarbor
   mipValueMode=TMVM_MipBias
   parameterName=Text01
   samplerSource=SSM_Clamp_WorldGroupSettings
   samplerType=SAMPLERTYPE_LinearColor
   texture=/BRCosmetics/Accessories/FORT_Backpacks/Backpack_M_MED_WeaveHarbor/Textures/T_WeaveHarbor_Backpack_Text01
  }
  n4 = TextureSampleParameter2D(UVs=@n1, Bias=@n2) {
   group=WeaveHarbor
   mipValueMode=TMVM_MipBias
   parameterName=Text02
   samplerSource=SSM_Clamp_WorldGroupSettings
   samplerType=SAMPLERTYPE_LinearColor
   texture=/BRCosmetics/Accessories/FORT_Backpacks/Backpack_M_MED_WeaveHarbor/Textures/T_WeaveHarbor_Backpack_Text02
  }
  n5 = FunctionInput() {bUsePreviewValueAsDefault=true, inputName=Phase, inputType=FunctionInput_Scalar, sortPriority=5}
  n6 = TextureSampleParameter2D(UVs=@n1, Bias=@n2) {
   group=WeaveHarbor
   mipValueMode=TMVM_MipBias
   parameterName=Text03
   samplerSource=SSM_Clamp_WorldGroupSettings
   samplerType=SAMPLERTYPE_LinearColor
   texture=/BRCosmetics/Accessories/FORT_Backpacks/Backpack_M_MED_WeaveHarbor/Textures/T_WeaveHarbor_Backpack_Text03
  }
  n7 = LinearInterpolate(A=LinearInterpolate(A=@n3.RGBA, B=@n4.RGBA, Alpha=Saturate(None=@n5)), B=@n6.RGBA, Alpha=Saturate(None=Subtract(A=@n5)))
  FunctionOutput(None=ComponentMask(None=@n7) {b=true, g=true, r=true}) {outputName=RGB}
  n8 = MaterialFunctionCall(Float4=@n7) {materialFunction=&BreakOutFloat4Components}
  FunctionOutput(None=@n8.R) {outputName=R, sortPriority=1}
  FunctionOutput(None=@n8.G) {outputName=G, sortPriority=2}
  FunctionOutput(None=@n8.B) {outputName=B, sortPriority=3}
  FunctionOutput(None=@n8.A) {outputName=A, sortPriority=4}
 ]}
