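# Parent material for the WeaveHarbor backpack's cloud FX on Niagara particles and meshes, drawing toon-lit clouds through emissive on Cloth shading.
# Panning height and normal textures shape and displace the puffs, with two dynamic-parameter "feet" spheres denting them.
# Layered shadow, rim, GI and AO colors follow the sky light, with PropMode swapping particle inputs for time-of-day tint and object scale.
# A HealthPhase-selected text overlay tints one side, growing, rotating and blurring as it fades.
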Material /BRCosmetics/Accessories/FORT_Backpacks/Backpack_M_MED_WeaveHarbor/FX/M_Backpack_WeaveHarbor_Cloud_Parent
graph {
 properties {
  allowTranslucentCustomDepthWrites=true
  bEnableResponsiveAA=true
  bUseMaterialAttributes=true
  bUsedWithNiagaraMeshParticles=true
  bUsedWithNiagaraSprites=true
  pixelDepthOffsetMode=PDOM_Legacy
  shadingModel=MSM_Cloth
  usedShadingModels=Cloth
 }
 nodes [
  n1 = DynamicParameter() {parameterIndex=1, defaultValue=<rgba 1 1 1 0>, paramNames [FeetPosX, FeetPosY, FeetPosZ, FeetHoleColor]}
  n2 = DynamicParameter() {parameterIndex=1, defaultValue=<rgba 1 1 1 0>, paramNames [FeetPosX, FeetPosY, FeetPosZ, FeetHoleColor]}
  n3 = AppendVector(A=@n2.FeetPosX, B=@n2.FeetPosX)
  n4 = WorldPosition() {worldPositionShaderOffset=WPT_ExcludeAllShaderOffsets}
  n5 = VectorParameter() {group=70.Feet, parameterName=FeetHole_SizeAndHardness, defaultValue=<rgba 30 0.5 0 0>}
  n6 = DynamicParameter() {parameterIndex=2, paramNames [Feet2PosX,Feet2PosY,Feet2PosZ,Grow]}
  n7 = AppendVector(A=@n6.Feet2PosX, B=@n6.Feet2PosX)
  n8 = VectorParameter() {group=70.Feet, parameterName=FeetHole1_SizeAndHardness, defaultValue=<rgba 30 0.5 0 0>}
  n9 = Max(
   A=SphereMask(A=AppendVector(A=@n3, B=@n2.FeetPosZ), B=@n4.XYZ, Radius=@n5.R, Hardness=@n5.R)
   B=SphereMask(A=AppendVector(A=@n7, B=@n6.Feet2PosZ), B=@n4.XYZ, Radius=@n8.R, Hardness=@n8.R)
  )
  n10 = OneMinus(None=@n9)
  LinearInterpolate(B=Multiply(A=@n10, B=Constant() {r=1}), Alpha=@n1.FeetHoleColor) {constA=1}
  n11 = OneMinus(None=NamedRerouteUsage())
  n12 = ScalarParameter() {parameterName=WPO_Add_Power}
  n13 = Constant3Vector() {constant=<rgba 0 0 1 0>}
  Multiply(
   A=Multiply(A=Power(Base=@n11, Exp=@n12), B=Transform(None=@n13) {transformSourceType=TRANSFORMSOURCE_Instance})
   B=ScalarParameter() {parameterName=WPO_Add_Center}
  )
  n14 = Multiply(A=Time(), B=ScalarParameter() {defaultValue=1, group=00.Base, parameterName=Base _TimeNodeSpeed})
  n15 = DynamicParameter() {paramNames [ScaleX,Age,Param3,Param4]}
  NamedRerouteDeclaration(None=Multiply(A=Add(A=@n14, B=@n15.Age), B=ScalarParameter() {defaultValue=1, group=00.Base, parameterName=Base_TimeWPO_Scale})) {name=Base_WPOTime, nodeColor=<rgba 1 0.5783 0 1>}
  n16 = MaterialFunctionCall(Float2=TextureCoordinate()) {materialFunction=&BreakOutFloat2Components}
  n17 = ScalarParameter() {defaultValue=1, parameterName=WPO_Main_VPower}
  NamedRerouteDeclaration(None=MaterialFunctionCall(X=@n16.R, Y=Power(Base=@n16.G, Exp=@n17)) {materialFunction=&MakeFloat2}.Result) {name=Base_WPOUV, nodeColor=<rgba 0.0312 1 0 1>}
  n18 = MaterialFunctionCall(Float2=TextureCoordinate()) {materialFunction=&BreakOutFloat2Components}
  n19 = Sine(None=Multiply(A=@n18.G) {constB=0.5})
  n20 = Power(Base=@n19, Exp=ScalarParameter() {defaultValue=2, group=20.Mask, parameterName=Mask_ShapeTex_A_UpPower})
  n21 = Power(Base=@n19, Exp=ScalarParameter() {defaultValue=5, group=20.Mask, parameterName=Mask_ShapeTex_A_DownPower})
  n22 = Saturate(None=Multiply(A=Subtract(A=@n18.G) {constB=0.475}) {constB=20})
  NamedRerouteDeclaration(None=StaticSwitchParameter(True=Constant() {r=1}, False=LinearInterpolate(A=@n20, B=@n21, Alpha=@n22)) {parameterName=CloudPartMod}) {name=Mask_ShapeTex_A, nodeColor=<rgba 1 0 0.4219 1>}
  n23 = MaterialFunctionCall(Float2=TextureCoordinate()) {materialFunction=&BreakOutFloat2Components}
  n24 = Sine(None=Multiply(A=@n23.G) {constB=0.5})
  n25 = ScalarParameter() {defaultValue=2, group=20.Mask, parameterName=Mask_ShapeTex_B_UpPower}
  n26 = ScalarParameter() {defaultValue=5, group=20.Mask, parameterName=Mask_ShapeTex_B_DownPower}
  n27 = Saturate(None=Multiply(A=Subtract(A=@n23.G) {constB=0.475}) {constB=20})
  NamedRerouteDeclaration(None=LinearInterpolate(A=Power(Base=@n24, Exp=@n25), B=Power(Base=@n24, Exp=@n26), Alpha=@n27)) {name=Mask_ShapeTex_B, nodeColor=<rgba 1 0.2405 0 1>}
  n28 = TextureObjectParameter() {
   group=10.Cloud_Main
   parameterName=Cloud_Main_HeightTex
   samplerType=SAMPLERTYPE_LinearGrayscale
   texture=/Game/Gadgets/Assets/VinderTech_GliderChute/Glider_Thrive_Female/FX/Texture/T_CloudBubble
  }
  n29 = VectorParameter() {group=10.Cloud_Main, parameterName=Cloud_Main_UVOffset}
  n30 = VectorParameter() {group=10.Cloud_Main, parameterName=Cloud_Main_UVControl, defaultValue=<rgba 1 1 0 0>}
  n31 = NamedRerouteUsage()
  n32 = NamedRerouteUsage()
  n33 = Add(
   A=Power(
    Base=MaterialFunctionCall(
     Texture=@n28
     AdditionalUVOffset=@n29.RGB
     UVControl_RGB=@n30.RGB
     UVControl_A=@n30.RGB
     "[Override]TimeInput"=@n31
     "[Override]UVInput"=@n32
    ) {materialFunction=&FXCommon_TextureSample}.R
    Exp=ScalarParameter() {defaultValue=1, group=10.Cloud_Main, parameterName=Cloud_Main_WPOPower}
   )
   B=ScalarParameter() {defaultValue=-0.5, group=10.Cloud_Main, parameterName=Cloud_Main_WPOStrengthOffset}
  ) {constB=-0.5}
  n34 = ScalarParameter() {defaultValue=100, group=10.Cloud_Main, parameterName=Cloud_Main_WPOStrength}
  NamedRerouteDeclaration(None=Multiply(A=Multiply(A=@n33, B=@n34), B=VertexNormalWS())) {name=Cloud_Main_WPO, nodeColor=<rgba 1 0 0.7266 1>}
  n35 = Constant3Vector() {constant=<rgba 0 0 1 0>}
  n36 = NamedRerouteUsage()
  n37 = TextureObjectParameter() {
   group=10.Cloud_Main
   parameterName=Cloud_Main_NormalTex
   samplerType=SAMPLERTYPE_Normal
   texture=/Game/Gadgets/Assets/VinderTech_GliderChute/Glider_Thrive_Female/FX/Texture/T_CloudBubble_N
  }
  n38 = MaterialFunctionCall(
   Texture=@n37
   AdditionalUVOffset=@n29.RGB
   UVControl_RGB=@n30.RGB
   UVControl_A=@n30.RGB
   "[Override]TimeInput"=@n31
   "[Override]UVInput"=@n32
  ) {materialFunction=&FXCommon_TextureSample}
  NamedRerouteDeclaration(None=StaticSwitchParameter(True=Constant3Vector() {constant=<rgba 0 0 1 0>}, False=LinearInterpolate(A=@n35, B=@n38.RGB, Alpha=@n36)) {parameterName=CloudPartMod}) {name=Cloud_Main_Normal, nodeColor=<rgba 0 0.5427 1 1>}
  n39 = VectorParameter() {group=29.Color, parameterName=Color_OutShadow, defaultValue=<rgba 0 0 0 1>}
  n40 = VectorParameter() {group=29.Color, parameterName=Color_OutShadow_Rim, defaultValue=<rgba 0 0 0 1>}
  n41 = Transform(None=NamedRerouteUsage())
  n42 = CameraVectorWS()
  n43 = Saturate(None=OneMinus(None=DotProduct(A=@n41, B=@n42)))
  n44 = ScalarParameter() {defaultValue=5, group=39.RimLight, parameterName=RimLight_POW}
  n45 = Multiply(A=NamedRerouteUsage(), B=Saturate(None=Power(Base=@n43, Exp=@n44)))
  n46 = NamedRerouteUsage()
  n47 = ScalarParameter() {group=30.LightingAnimeShadow, parameterName=Lighting_AnimeShadow_NormalFlat}
  n48 = DotProduct(
   A=NamedRerouteUsage()
   B=Transform(None=MaterialFunctionCall(Normal=@n46, Flatness=@n47) {materialFunction=&FlattenNormal}.Result)
  )
  n49 = ScalarParameter() {group=30.LightingAnimeShadow, parameterName=Lighting_AnimeShadow_NDS_Offset}
  n50 = Saturate(None=Multiply(A=Add(A=@n48, B=@n49), B=ScalarParameter() {defaultValue=1, group=30.LightingAnimeShadow, parameterName=Lighting_AnimeShadow_NDS_Mul}))
  n51 = ScalarParameter() {defaultValue=1, group=30.LightingAnimeShadow, parameterName=Lighting_AnimeShadow_NDS_Power}
  n52 = VectorParameter() {group=30.LightingAnimeShadow, parameterName=Lighting_AnimeShadow_NDS_Remap, defaultValue=<rgba 0 1 0 1>}
  n53 = MaterialFunctionCall(Input=Saturate(None=Power(Base=@n50, Exp=@n51)), Input Low=Constant(), Input High=Constant() {r=1}, Target Low=@n52.R, Target High=@n52.R) {materialFunction=&RemapValueRange}
  n54 = NamedRerouteUsage()
  n55 = Transform(None=MaterialFunctionCall(Normal=NamedRerouteUsage(), Flatness=ScalarParameter() {group=32.LightingGI, parameterName=Lighting_GI_NormalFlat}) {materialFunction=&FlattenNormal}.Result)
  n56 = Multiply(
   A=Add(A=DotProduct(A=@n54, B=@n55), B=ScalarParameter() {group=32.LightingGI, parameterName=Lighting_GI_NDS_Offset})
   B=ScalarParameter() {defaultValue=1, group=32.LightingGI, parameterName=Lighting_GI_NDS_Mul}
  )
  n57 = Saturate(None=Power(Base=Saturate(None=@n56), Exp=ScalarParameter() {defaultValue=1, group=32.LightingGI, parameterName=Lighting_GI_NDS_Power}))
  n58 = Constant()
  n59 = Constant() {r=1}
  n60 = VectorParameter() {group=32.LightingGI, parameterName=Lighting_GI_NDS_Remap, defaultValue=<rgba 0 1 0 1>}
  n61 = LinearInterpolate(
   A=VectorParameter() {group=29.Color, parameterName=Color_GI, defaultValue=<rgba 0 0 0 1>}.RGB
   B=LinearInterpolate(
    A=VectorParameter() {group=29.Color, parameterName=Color_InShadow, defaultValue=<rgba 0 0 0 1>}.RGB
    B=LinearInterpolate(A=@n39.RGB, B=@n40.RGB, Alpha=@n45)
    Alpha=Saturate(None=@n53.Result)
   )
   Alpha=Saturate(None=MaterialFunctionCall(Input=@n57, Input Low=@n58, Input High=@n59, Target Low=@n60.R, Target High=@n60.R) {materialFunction=&RemapValueRange}.Result)
  )
  n62 = Desaturation(
   None=MaterialFunctionCall(Hue Shift Percentage=ScalarParameter() {group=31.LightingB, parameterName=Lighting_AO_HueShift}, Texture=@n61) {materialFunction=&HueShift}.Result
   Fraction=ScalarParameter() {group=31.LightingB, parameterName=Lighting_AO_Desaturation}
  )
  n63 = MaterialFunctionCall(Normal=NamedRerouteUsage(), Flatness=ScalarParameter() {group=31.LightingAO, parameterName=Lighting_AO_NormalFlat}) {materialFunction=&FlattenNormal}
  n64 = Add(
   A=DotProduct(A=NamedRerouteUsage(), B=Transform(None=@n63.Result))
   B=ScalarParameter() {group=31.LightingAO, parameterName=Lighting_AO_NDS_Offset}
  )
  n65 = ScalarParameter() {defaultValue=1, group=31.LightingAO, parameterName=Lighting_AO_NDS_Mul}
  n66 = Power(
   Base=Saturate(None=Multiply(A=@n64, B=@n65))
   Exp=ScalarParameter() {defaultValue=1, group=31.LightingAO, parameterName=Lighting_AO_NDS_Power}
  )
  n67 = VectorParameter() {group=31.LightingAO, parameterName=Lighting_AO_NDS_Remap, defaultValue=<rgba 0 1 0 1>}
  n68 = Saturate(None=MaterialFunctionCall(Input=Saturate(None=@n66), Input Low=Constant(), Input High=Constant() {r=1}, Target Low=@n67.R, Target High=@n67.R) {materialFunction=&RemapValueRange}.Result)
  NamedRerouteDeclaration(None=Multiply(A=LinearInterpolate(A=@n62, B=@n61, Alpha=@n68), B=@n68)) {name=LightingOutEmissive, nodeColor=<rgba 0.0547 0 1 1>}
  NamedRerouteDeclaration(None=LinearInterpolate(A=@n20, Alpha=@n22)) {name=Mask_ShapeTex_A_JustUp, nodeColor=<rgba 1 0 0.4219 1>}
  n69 = VectorParameter() {parameterName=FELightDir}
  NamedRerouteDeclaration(None=LinearInterpolate(A=SkyAtmosphereLightDirection(), B=@n69.RGB, Alpha=@n69.RGB)) {name=LightDir, nodeColor=<rgba 0 0.2031 1 1>}
  ScalarParameter() {defaultValue=3, parameterName=TextAppearDelay}
  n70 = Constant()
  n71 = LinearInterpolate(A=NamedRerouteUsage(), B=Constant3Vector() {constant=<rgba 0 0 0 1>})
  n72 = MaterialFunctionCall(Hue Shift Percentage=Constant(), Texture=@n71) {materialFunction=&HueShift}
  n73 = MaterialFunctionCall(Value_Morning=Constant() {r=-0.1}, Value_Day=Constant(), Value_Evening=Constant() {r=-0.1}, Value_Night=Constant() {r=0.05}) {materialFunction=&Scalar_Time_of_Day}
  n74 = Constant() {r=0.3}
  n75 = Constant() {r=1}
  n76 = Constant() {r=0.3}
  n77 = Constant() {r=1}
  n78 = DynamicParameter() {parameterIndex=3, defaultValue=<rgba 0 0 1 1>, paramNames [Hue,Desat,Mul,Param4]}
  n79 = MaterialFunctionCall(Hue Shift Percentage=@n78.Hue, Texture=@n71) {materialFunction=&HueShift}
  n80 = StaticSwitchParameter(
   True=Multiply(
    A=Desaturation(None=@n72.Result, Fraction=@n73.Result)
    B=MaterialFunctionCall(Value_Morning=@n74, Value_Day=@n75, Value_Evening=@n76, Value_Night=@n77) {materialFunction=&Scalar_Time_of_Day}.Result
   )
   False=Multiply(A=Desaturation(None=@n79.Result, Fraction=@n78.Desat), B=@n78.Mul)
  ) {parameterName=PropMode}
  n81 = NamedRerouteUsage()
  n82 = ScalarParameter() {defaultValue=-4, group=00.Base, parameterName=Outline_Normalflat}
  n83 = ScalarParameter() {defaultValue=2, group=20.Mask, parameterName=Mask_ShapeTex_B_UpWPOPower}
  n84 = Multiply(
   A=LinearInterpolate(A=Power(Base=@n24, Exp=@n83), Alpha=@n27)
   B=Transform(None=VectorParameter() {parameterName=EdgeWPO, defaultValue=<rgba 0 0 10 1>}.RGB) {transformSourceType=TRANSFORMSOURCE_Local}
  )
  n85 = StaticSwitchParameter(
   True=Multiply(A=MaterialFunctionCall() {materialFunction=&ObjectScale}.Scale X) {constB=2}
   False=DynamicParameter() {paramNames [ScaleX,Age,Param3,Param4]}.ScaleX
  ) {parameterName=PropMode}
  n86 = Transform(None=Constant3Vector() {constant=<rgba 0 0 -1 0>}) {transformSourceType=TRANSFORMSOURCE_Local}
  n87 = Multiply(A=Transform(None=NamedRerouteUsage()), B=ScalarParameter() {parameterName=WPO_Main_NormalStrength})
  n88 = Multiply(A=NamedRerouteUsage(), B=NamedRerouteUsage())
  n89 = Add(
   A=Multiply(A=Multiply(A=@n9, B=@n86), B=ScalarParameter() {group=70.Feet, parameterName=Feet_WPOStrength})
   B=Multiply(A=@n10, B=Add(A=@n87, B=@n88))
  )
  n90 = StaticSwitchParameter(True=Constant(), False=Add(A=Multiply(A=@n84, B=@n10), B=Multiply(A=@n89, B=@n85))) {parameterName=CloudPartMod}
  n91 = ScalarParameter() {defaultValue=1, parameterName=PostWPOStrengthScale}
  n92 = GetMaterialAttributes(None=MakeMaterialAttributes!(
    BaseColor=@n70
    Metallic
    Specular=@n70
    Roughness=Constant() {r=0.237}
    Anisotropy
    EmissiveColor=MaterialFunctionCall(Emissive=@n80) {materialFunction=&MF_FX_EmissiveAdjustments}.Result
    Opacity=Constant() {r=1}
    OpacityMask
    Normal=MaterialFunctionCall(Normal=@n81, Flatness=@n82) {materialFunction=&FlattenNormal}.Result
    Tangent
    WorldPositionOffset=Multiply(A=@n90, B=@n91)
    SubsurfaceColor=Constant() {r=0.005}
    ClearCoat
    ClearCoatRoughness
    AmbientOcclusion
    Refraction
    CustomizedUVs_0
    CustomizedUVs_1
    CustomizedUVs_2
    CustomizedUVs_3
    CustomizedUVs_4
    CustomizedUVs_5
    CustomizedUVs_6
    CustomizedUVs_7
    PixelDepthOffset
    ShadingModel
    Displacement
   )) {attributeGetTypes [#1]}
  n93 = VectorParameter() {parameterName=Text_Tint, defaultValue=<rgba 0.4423 0.4297 0.4688 1>}
  n94 = ScalarParameter() {parameterName=TextFadeStartTime}
  n95 = Time()
  n96 = MaterialFunctionCall(
   TimestampA=If(A=@n94, B=@n95, "A > B"=Add(A=@n94) {constB=-100}, "A < B"=@n94)
   TransitionLength_In=ScalarParameter() {defaultValue=4, parameterName=TextFadeDuration}
   Time=@n95
  ) {materialFunction=&MF_Timestamp_InOut_01}
  n97 = Power(Base=@n96.Linear) {constExponent=0.5}
  n98 = Multiply(A=ScalarParameter() {defaultValue=5, parameterName=Text_MaximumMipBias}, B=@n97)
  n99 = ScalarParameter() {defaultValue=450, parameterName=Text_TextureScale}
  n100 = ScalarParameter() {defaultValue=1000, parameterName=Text_TextureScale_End}
  n101 = LocalPosition()
  n102 = AppendVector(
   A=DotProduct(A=@n101.XYZ, B=Constant3Vector() {constant=<rgba 0 1 0 0>})
   B=DotProduct(A=Constant3Vector() {constant=<rgba 1 0 0 0>}, B=@n101.XYZ)
  )
  n103 = ScalarParameter() {defaultValue=0.5, parameterName=Text_CenterU, sliderMax=1}
  n104 = ScalarParameter() {defaultValue=0.5, parameterName=Text_CenterV, sliderMax=1}
  n105 = ScalarParameter() {defaultValue=-0.04, parameterName=Text_RotationAngle}
  n106 = ScalarParameter() {defaultValue=-0.08, parameterName=Text_RotationAngle_End}
  n107 = MaterialFunctionCall!(
   Texture Scale=Multiply(A=Constant2Vector() {g=1.25, r=1}, B=LinearInterpolate(A=@n99, B=@n100, Alpha=@n97))
   UVs=MaterialFunctionCall(
    UVs=OneMinus(None=@n102)
    Rotation Center=AppendVector(A=@n103, B=@n104)
    "Rotation Angle (0-1)"=LinearInterpolate(A=@n105, B=@n106, Alpha=@n97)
   ) {materialFunction=&CustomRotator}.Rotated Values
   Offset XY Texturesize Z=MaterialFunctionCall(X=@n103, Y=@n104, Z=Constant() {r=1}) {materialFunction=&MakeFloat3}.Result
  ) {materialFunction=/Engine/Functions/Engine_MaterialFunctions01/Texturing/ScaleUVsAroundPoint}
  n108 = ScalarParameter() {parameterName=HealthPhase}
  n109 = Multiply(
   A=Multiply(
    A=MaterialFunctionCall!(MipBias=@n98, UV=@n107.UVs, Phase=@n108) {materialFunction=/BRCosmetics/Accessories/FORT_Backpacks/Backpack_M_MED_WeaveHarbor/Materials/MF_Backpack_WeaveHarbor_TextureSwap}.B
    B=@n107.0-1 mask
   )
   B=OneMinus(None=@n96.SmoothStep)
  )
  n110 = Constant3Vector() {constant=<rgba 0 -1 0 0>}
  n111 = Saturate(None=DotProduct(A=Transform(None=@n110) {transformSourceType=TRANSFORMSOURCE_Local}, B=VertexNormalWS()))
  SetMaterialAttributes(
   MaterialAttributes=@n92.MaterialAttributes
   Emissive Color=LinearInterpolate(A=@n92.Emissive Color, B=Multiply(A=@n92.Emissive Color, B=@n93.RGB), Alpha=Multiply(A=@n109, B=@n111))
  ) {attributeSetTypes [#1]}
  StaticBool() {value=true}
 ]
}
