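# Idle Niagara effect for the WeaveHarbor cosmetic backpack: a persistent cloud mesh attached to the skeletal mesh.
# LightningBase meshes and Decals sprites appear once User.HealthPercent drops to about half, spawning faster as it falls toward a quarter.
# User.LandBurst gates a ring burst of short-lived puff meshes.
# Cloud materials adapt to time of day and front end via dynamic parameters; the cloud's scale animation modules are disabled.
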NiagaraSystem /BRCosmetics/Accessories/FORT_Backpacks/Backpack_M_MED_WeaveHarbor/FX/NS_Backpack_WeaveHarbor_Idle
dependencies {
 usedDataInterfaces [
  &NiagaraDataInterfaceCurve
  &NiagaraDataInterfaceArrayColor
  &NiagaraDataInterfaceArrayFloat3
  &NiagaraDataInterfaceMeshRendererInfo
  &NiagaraDataInterfaceSkeletalMesh
  &NiagaraDataInterfaceArrayFloat
 ]
 usedDynamicInputs []
 usedModules [
  &self:SystemSpawnScript.SystemScriptSource.SystemScriptGraph.NiagaraNodeAssignment_0.SetVariables_#1
  &SystemState
  &Timeline
 ]
}
emitters {
 Decals {
  emitterUpdateScript [
   &EmitterState()
   &SpawnRate(
    Spawn Probability=0.6
    SpawnRate=&Multiply_Float(
     A=1
     B=&MakeCustomFloatFromBool(
      Bool=&SetBoolByFloatComparison(A=$User.HealthPercent, B=0.51, Comparison Type=::NewEnumerator2)
      True Float=0
      False Float=&Lerp_Float(
       A=0.3
       B=1
       Alpha=&RemapRange(Input Value To Remap=$User.HealthPercent, Input Min=0.5, Input Max=0.25, Output Min=0, Output Max=1)
      )
     )
    )
    Use Spawn Probability=true
   )
  ]
  particleSpawnScript [&InitializeParticle(
    Color=<rgba 0.5625 0.0103 0 0.5>
    Lifetime=0.45
    Sprite Size Mode=::NewEnumerator3
    Uniform Sprite Size=30
    UsePositionOffset=false
    Write Material Random=true
   ) {module=InitializeParticle001}]
  particleUpdateScript [
   &ParticleState()
   &SkeletalMeshLocation(Bone Sampling Mode=::NewEnumerator4, MaskSpawnGroup=false, Sampled Position Offset=<0 -6 -1>, Skeletal Mesh=$System.Skeletal Mesh)
   &Drag(Drag=10, Use Linear Drag=true, Use Rotational Drag=false)
   &SolveForcesAndVelocity(Clamp Velocity=false, Limit Acceleration=false)
   &CameraOffset(Camera Offset Amount=15)
   &DynamicMaterialParameters(
    Index 0 Param 1=1.5
    Index 0 Param 2=1
    Index 0 Param 3=-0.6
    Module.Param0WriteEnabled=true
    Module.Param1WriteEnabled=true
    Module.Param2WriteEnabled=true
    Module.Param3WriteEnabled=true
   )
   &ScaleColor(
    Scale Alpha=&FloatFromCurve(FloatCurve=DataInterface() {propertyValues {Curve {keys [
         <at 0 0 RCIM_Cubic>
         {arriveTangent=0.7215, interpMode=RCIM_Cubic, leaveTangent=0.7215, tangentMode=RCTM_User, time=0.2, value=0.8}
         {arriveTangent=0.0442, interpMode=RCIM_Cubic, leaveTangent=-1.1429, tangentMode=RCTM_Break, time=0.3261, value=0.1993}
         {arriveTangent=0.3826, interpMode=RCIM_Cubic, leaveTangent=0.3826, time=0.5, value=0.75}
         {arriveTangent=-0.175, interpMode=RCIM_Cubic, leaveTangent=-0.175, time=0.6404, value=0.3195}
         <at 0.7857 0.7>
         <at 1 0>
        ]}}})
    Scale RGB=&Multiply_VectorByFloat(Vector=<1 1 1>, Float=&FN_FrontEndFloatModifier(Gameplay=1, Front End=0.5))
    ScaleA=true
    ScaleRGB=true
   )
  ]
  rendererClasses [&NiagaraSpriteRendererProperties]
  renderers [{rendererClass=&NiagaraSpriteRendererProperties}]
 }
 LandingPuffCloud {
  emitterUpdateScript [
   &EmitterState(Life Cycle Mode=::NewEnumerator1, Loop Delay=0.01, Loop Duration=0.05, UseLoopDelay=false)
   &SpawnBurst_Instantaneous(
    Spawn Count=&RandomRangeInt(Minimum=4, Maximum=5)
    Spawn Probability=&MakeCustomFloatFromBool(
     Bool=&SetBoolByFloatComparison(A=$User.LandBurst, B=1, Comparison Type=::NewEnumerator2)
     True Float=1
     False Float=0
    )
    Use Spawn Probability=true
    UseLoopCountLimit=false
   )
  ]
  particleSpawnScript [
   &InitializeParticle(
    Lifetime Max=0.5
    Lifetime Min=0.2
    Lifetime Mode=::NewEnumerator1
    Mesh Scale=&Multiply_VectorByFloat(Vector=<0.01 0.02 0.02>, Float=&RandomRangeFloat(Minimum=0.7, Maximum=1.2))
    Mesh Scale Mode=::NewEnumerator4
    Sprite Size Mode=::NewEnumerator3
    Uniform Sprite Size=1
    UsePositionOffset=false
    Write Material Random=true
   )
   &SkeletalMeshLocation(Bone Sampling Mode=::NewEnumerator4, MaskSpawnGroup=false, Skeletal Mesh=$System.Skeletal Mesh) {enabled=false}
   &ShapeLocation(
    MaskSpawnGroup=false
    Offset=<0 -6 2>
    Offset Mode=::NewEnumerator0
    Override Seed=false
    Ring / Disc Distribution Mode:ENiagaraSphereDistributionMode=::NewEnumerator1
    Ring Radius=9.9
    Rotation Angle=0.25
    Rotation Axis=<0 1 0>
    Rotation Mode=::NewEnumerator1
    Shape Primitive=::NewEnumerator4
    U Position=&ReturnNormalizedExecIndex(
     Normalized Index Scale=&RandomRangeFloat(Minimum=0.9, Maximum=1.25)
     Normalized Index Includes One=false
    )
    UseOwnerScale=true
   ) {module=ShapeLocation001}
   &AddVelocity(
    Module.Use Falloff Distance=false
    Override Seed=false
    Velocity Mode=::NewEnumerator1
    Velocity Origin:Vector3f=$Particles.ShapeLocation001.ShapePosition
    Velocity Speed=&RandomRangeFloat(Minimum=150, Maximum=250)
   ) {enabled=false}
   &AddVelocity(
    Velocity=$Particles.ShapeLocation001.ShapePosition
    Velocity Speed Scale=&RandomRangeFloat(Minimum=2.5, Maximum=4)
   ) {module=AddVelocity001}
   &InitialMeshOrientation(Use Rotation Vector=false)
  ]
  particleUpdateScript [
   &Drag(Drag=7, Use Linear Drag=true, Use Rotational Drag=true)
   &SolveForcesAndVelocity(Clamp Velocity=false, Limit Acceleration=false)
   &ParticleState()
   &FN_Timer(Increment Multiplier:NiagaraFloat=&MakeCustomFloatFromBool(
     Bool=$User.IsFrontEnd
     True Float=&FloatFromCurve(
      FloatCurve=DataInterface() {propertyValues {Curve {keys [<at 0 2>,<at 0.4 2>,<at 1.6 0>]}}}
      CurveIndex=&Add_Float(A=$Particles.Age, B=&Multiply_FloatByInt(Float=100, Integer=$User.Deployed))
     )
     False Float=&Multiply_Float(A=&ClampFloat(Float=$System.InterpolateOverTime.InterpolateOverTime_Float.Moving Average), B=1)
    )) {enabled=false}
   &DynamicMaterialParameters(
    Index 0 Param 1=&Multiply_Float(
     A=&Multiply_Float(
      A=&MakeFloatFromVector(Vector=$Particles.Scale, Channel=::NewEnumerator0)
      B=&MakeFloatFromVector(Vector=$Engine.Owner.Scale, Channel=::NewEnumerator0)
     )
     B=2
    )
    Index 0 Param 3=&Power(a=&MakeFloatFromVector(Vector=$Engine.Owner.Scale, Channel=::NewEnumerator0), b=0.5)
    Index 3 Param 1=&FNTimeOfDayScalar(Night Scale=-0.0704, Morning Scale=-0.012, Day Scale=0, Evening Scale=0)
    Index 3 Param 2=&FNTimeOfDayScalar(Night Scale=-0.228, Morning Scale=0.2, Day Scale=0, Evening Scale=0)
    Index 3 Param 3=&MakeCustomFloatFromBool(
     Bool=$User.IsFrontEnd
     True Float=1.1
     False Float=&FNTimeOfDayScalar(Night Scale=0.588, Morning Scale=1, Evening Scale=1)
    )
    Module.Param0WriteEnabled=true
    Module.Param0WriteEnabled3=true
    Module.Param1WriteEnabled=false
    Module.Param1WriteEnabled3=true
    Module.Param2WriteEnabled=true
    Module.Param2WriteEnabled3=true
    Module.Param3WriteEnabled=false
    Module.Param3WriteEnabled3=true
    Write Parameter Index 3=true
   )
   &UpdateMeshOrientation(Rotation Rate=&RandomRangeFloat(Minimum=-1, Maximum=1)) {enabled=false}
   &ScaleMeshSize(Scale Factor=&VectorFromFloat(Value=&FloatFromCurve(FloatCurve=DataInterface() {propertyValues {Curve {keys [
          {arriveTangent=16.6204, interpMode=RCIM_Cubic, leaveTangent=16.6204, tangentMode=RCTM_User, time=0, value=0}
          {arriveTangent=0.2654, interpMode=RCIM_Cubic, leaveTangent=0.2654, tangentMode=RCTM_User, time=0.1636, value=1.476}
          {arriveTangent=-0.7988, interpMode=RCIM_Cubic, leaveTangent=-0.7988, time=0.2959, value=0.814}
          {arriveTangent=0.3306, interpMode=RCIM_Cubic, leaveTangent=0.3306, time=0.4349, value=1.2592}
          {arriveTangent=-0.8945, interpMode=RCIM_Cubic, leaveTangent=-0.8945, time=0.5567, value=0.9002}
          {arriveTangent=-0.5833, interpMode=RCIM_Cubic, leaveTangent=-0.5833, time=0.7325, value=0.993}
          {arriveTangent=-3.7129, interpMode=RCIM_Cubic, leaveTangent=-3.7129, time=0.9, value=0.7}
          {arriveTangent=-3.986, interpMode=RCIM_Cubic, leaveTangent=-3.986, tangentMode=RCTM_User, time=1, value=0}
         ]}}}))) {module=ScaleMeshSize001}
  ]
  rendererClasses [&NiagaraMeshRendererProperties]
  renderers [{rendererClass=&NiagaraMeshRendererProperties, rendererIndex=0}]
 }
 LightningBase {
  emitterUpdateScript [
   &EmitterState()
   &SpawnRate(
    SpawnRate=&Multiply_Float(
     A=3
     B=&MakeCustomFloatFromBool(
      Bool=&SetBoolByFloatComparison(A=$User.HealthPercent, B=0.51, Comparison Type=::NewEnumerator2)
      True Float=0
      False Float=&Lerp_Float(
       A=0.3
       B=1.1
       Alpha=&RemapRange(Input Value To Remap=$User.HealthPercent, Input Min=0.5, Input Max=0.25, Output Min=0, Output Max=1)
      )
     )
    )
    Use Spawn Probability=true
   )
  ]
  particleSpawnScript [
   &InitializeParticle(
    Color=&SelectLinearColorFromArray(
     Array Sampling Mode=::NewEnumerator0
     Direct Array Index=?
     Color Selection Array=DataInterface() {propertyValues {ColorData [<rgba 0.6875 0.0058 0.0469 1>, <rgba 0.0995 0.0264 0.526 1>]}}
     Random Seed=0
    )
    Lifetime Max=1.2
    Lifetime Min=0.5
    Lifetime Mode=::NewEnumerator1
    Mesh Renderer Array Visibility Mode=::NewEnumerator3
    Mesh Renderer Info=DataInterface() {propertyValues {MeshRenderer=&self:LightningBase_1.NiagaraMeshRendererProperties_0}}
    Mesh Scale=<0.04 0.04 0.2>
    Mesh Scale Mode=::NewEnumerator4
    UsePositionOffset=false
    Write Material Random=true
   )
   &ShapeLocation(
    Hemisphere Distribution=<0.5 1>
    MaskSpawnGroup=false
    Offset=<0 -8 0>
    Offset Mode=::NewEnumerator0
    Override Seed=false
    Rotation Mode=::NewEnumerator2
    Sphere Radius=1
    UseOwnerScale=true
    Yaw / Pitch / Roll=<0 0 0.5>
   )
   &InitialMeshOrientation(
    Mesh Orientation Mode=::NewEnumerator2
    Orientation Vector=$Particles.ShapeLocation.ShapeNormal
    Rotation=&RandomRangeVector(Minimum=<0 0 0>, Maximum=<0 0 1>)
    Use Rotation Vector=true
   )
   &CameraOffset(Camera Offset Amount=4)
  ]
  particleUpdateScript [
   &SkeletalMeshLocation(
    MaskSpawnGroup=false
    Mesh Sampling Type=::NewEnumerator5
    Sampled Position Offset=<0 -5 -120>
    Skeletal Mesh=$System.Skeletal Mesh
    Socket Sampling Mode=::NewEnumerator5
   ) {enabled=false}
   &SolveForcesAndVelocity(Clamp Velocity=false, Limit Acceleration=false)
   &ParticleState()
   &ScaleColor(
    Scale Alpha=&FloatFromCurve(FloatCurve=DataInterface() {propertyValues {Curve {keys [
         {interpMode=RCIM_Cubic, leaveTangent=0.2357, tangentMode=RCTM_Break, time=0.3, value=1}
         {arriveTangent=-1.6327, tangentMode=RCTM_User, time=1, value=0}
        ]}}})
    Scale RGB=&VectorFromFloat(Value=&FloatFromCurve(FloatCurve=DataInterface() {propertyValues {Curve {keys [
          <at 0 1>
          {arriveTangent=1.25, leaveTangent=1.25, time=0.3, value=1}
          {interpMode=RCIM_Cubic, leaveTangent=-5.9679, tangentMode=RCTM_Break, time=0.4, value=1.5}
          <at 1 0.4>
         ]}}}))
    ScaleA=true
    ScaleRGB=true
   )
   &DynamicMaterialParameters(
    Index 0 Param 1=&FloatFromCurve(FloatCurve=DataInterface() {propertyValues {Curve {keys [{interpMode=RCIM_Cubic, tangentMode=RCTM_User, time=0, value=0}, {interpMode=RCIM_Cubic, tangentMode=RCTM_User, time=1, value=1}]}}})
    Index 0 Param 2=&RandomRangeFloat(Minimum=4, Maximum=8)
    Index 0 Param 3=&FloatFromCurve(
     FloatCurve=DataInterface() {propertyValues {Curve {keys [
         <at 0 1>
         {interpMode=RCIM_Cubic, leaveTangent=0.1085, tangentMode=RCTM_Break, time=0.1038, value=1}
         {interpMode=RCIM_Cubic, leaveTangent=-8.5033, tangentMode=RCTM_Break, time=0.2038, value=2.625}
         <at 1 0>
        ]}}}
     Scale Curve=&RandomRangeFloat(Minimum=15, Maximum=20)
    )
    Module.Index0Param0Write=true
    Module.Index0Param1Write=true
    Module.Index0Param2Write=true
    Module.Index0Param3Write=true
   )
   /SetVariables(Particles.Dy2:Vector4f=&MakeVector4(
     X=&FloatFromCurve(FloatCurve=DataInterface() {propertyValues {Curve {keys [{interpMode=RCIM_Cubic, tangentMode=RCTM_User, time=0, value=0}, {interpMode=RCIM_Cubic, tangentMode=RCTM_User, time=1, value=1}]}}})
     Y=&RandomRangeFloat(Minimum=4, Maximum=12)
     Z=&FloatFromCurve(
      FloatCurve=DataInterface() {propertyValues {Curve {keys [
          <at 0 1>
          {interpMode=RCIM_Cubic, leaveTangent=0.1085, tangentMode=RCTM_Break, time=0.1038, value=1}
          {interpMode=RCIM_Cubic, leaveTangent=-8.5033, tangentMode=RCTM_Break, time=0.2038, value=2.625}
          <at 1 0>
         ]}}}
      Scale Curve=&RandomRangeFloat(Minimum=15, Maximum=20)
     )
     W=0
    )) {isSetParameters=true, module=SetVariables_#2}
  ]
  rendererClasses [&NiagaraMeshRendererProperties]
  renderers [
   {rendererClass=&NiagaraMeshRendererProperties, rendererIndex=0}
   {rendererClass=&NiagaraMeshRendererProperties, rendererIndex=1}
  ]
 }
 LightningBase003 {
  emitterUpdateScript [
   &EmitterState()
   &SpawnRate(
    SpawnRate=&Multiply_Float(
     A=1.5
     B=&MakeCustomFloatFromBool(
      Bool=&SetBoolByFloatComparison(A=$User.HealthPercent, B=0.51, Comparison Type=::NewEnumerator2)
      True Float=0
      False Float=&Lerp_Float(
       A=0.3
       B=1
       Alpha=&RemapRange(Input Value To Remap=$User.HealthPercent, Input Min=0.5, Input Max=0.25, Output Min=0, Output Max=1)
      )
     )
    )
    Use Spawn Probability=true
   )
  ]
  particleSpawnScript [
   &InitializeParticle(
    Color=&SelectLinearColorFromArray(
     Array Sampling Mode=::NewEnumerator0
     Direct Array Index=?
     Color Selection Array=DataInterface() {propertyValues {ColorData [<rgba 0.6875 0.0058 0.0469 1>, <rgba 0.0995 0.0264 0.526 1>]}}
     Random Seed=0
    )
    Lifetime Max=1.25
    Lifetime Min=0.7
    Lifetime Mode=::NewEnumerator1
    Mesh Renderer Array Visibility Mode=::NewEnumerator3
    Mesh Renderer Info=DataInterface() {propertyValues {MeshRenderer=&self:LightningBase003_11.NiagaraMeshRendererProperties_0}}
    Mesh Scale=&Multiply_Vector(
     A=<0.36 0.36 0.81>
     B=&SelectVectorFromArray(
      Array Sampling Mode=::NewEnumerator0
      Direct Array Index=?
      Vector Selection Array=DataInterface() {propertyValues {FloatData [<-1 -1 1>, <1 1 1>]}}
      Random Seed=0
     )
    )
    Mesh Scale Mode=::NewEnumerator4
    UsePositionOffset=false
    Write Material Random=true
   )
   &InitialMeshOrientation(
    Mesh Orientation Mode=::NewEnumerator2
    Rotation=&RandomRangeVector(Minimum=<0 0 0>, Maximum=<0 0 1>)
    Use Rotation Vector=true
   )
   &CameraOffset(Camera Offset Amount=4)
  ]
  particleUpdateScript [
   &SkeletalMeshLocation(
    MaskSpawnGroup=false
    Mesh Sampling Type=::NewEnumerator5
    Sampled Position Offset=<0 -6 -100>
    Skeletal Mesh=$System.Skeletal Mesh
    Socket Sampling Mode=::NewEnumerator5
   )
   &SolveForcesAndVelocity(Clamp Velocity=false, Limit Acceleration=false)
   &ParticleState()
   &ScaleColor(
    Scale Alpha=&FloatFromCurve(FloatCurve=DataInterface() {propertyValues {Curve {keys [
         {interpMode=RCIM_Cubic, leaveTangent=0.2357, tangentMode=RCTM_Break, time=0.3, value=1}
         {arriveTangent=-1.6327, tangentMode=RCTM_User, time=1, value=0}
        ]}}})
    Scale RGB=&VectorFromFloat(Value=&FloatFromCurve(FloatCurve=DataInterface() {propertyValues {Curve {keys [
          <at 0 1>
          {arriveTangent=1.25, leaveTangent=1.25, time=0.3, value=1}
          {interpMode=RCIM_Cubic, leaveTangent=-5.9679, tangentMode=RCTM_Break, time=0.4, value=1.5}
          <at 1 0.4>
         ]}}}))
    ScaleA=true
    ScaleRGB=true
   )
   &DynamicMaterialParameters(
    Index 0 Param 1=&FloatFromCurve(FloatCurve=DataInterface() {propertyValues {Curve {keys [{interpMode=RCIM_Cubic, tangentMode=RCTM_User, time=0, value=0}, {interpMode=RCIM_Cubic, tangentMode=RCTM_User, time=1, value=1}]}}})
    Index 0 Param 2=&RandomRangeFloat(Minimum=4, Maximum=8)
    Index 0 Param 3=&FloatFromCurve(
     FloatCurve=DataInterface() {propertyValues {Curve {keys [
         <at 0 1>
         {interpMode=RCIM_Cubic, leaveTangent=0.1085, tangentMode=RCTM_Break, time=0.1038, value=1}
         {interpMode=RCIM_Cubic, leaveTangent=-8.5033, tangentMode=RCTM_Break, time=0.2038, value=2.625}
         <at 1 0>
        ]}}}
     Scale Curve=&RandomRangeFloat(Minimum=15, Maximum=20)
    )
    Module.Index0Param0Write=true
    Module.Index0Param1Write=true
    Module.Index0Param2Write=true
    Module.Index0Param3Write=true
   )
   /SetVariables(Particles.Dy2:Vector4f=&MakeVector4(
     X=&FloatFromCurve(FloatCurve=DataInterface() {propertyValues {Curve {keys [{interpMode=RCIM_Cubic, tangentMode=RCTM_User, time=0, value=0}, {interpMode=RCIM_Cubic, tangentMode=RCTM_User, time=1, value=1}]}}})
     Y=&RandomRangeFloat(Minimum=8, Maximum=16)
     Z=&FloatFromCurve(
      FloatCurve=DataInterface() {propertyValues {Curve {keys [
          <at 0 1>
          {interpMode=RCIM_Cubic, leaveTangent=0.1085, tangentMode=RCTM_Break, time=0.1038, value=1}
          {interpMode=RCIM_Cubic, leaveTangent=-8.5033, tangentMode=RCTM_Break, time=0.2038, value=2.625}
          <at 1 0>
         ]}}}
      Scale Curve=&RandomRangeFloat(Minimum=15, Maximum=20)
     )
     W=0
    )) {isSetParameters=true, module=SetVariables_#2}
  ]
  rendererClasses [&NiagaraMeshRendererProperties]
  renderers [
   {rendererClass=&NiagaraMeshRendererProperties, rendererIndex=0}
   {rendererClass=&NiagaraMeshRendererProperties, rendererIndex=1}
  ]
 }
 MainCloud {
  emitterUpdateScript [
   &EmitterState(
    Inactive Response=::NewEnumerator1
    Life Cycle Mode=::NewEnumerator1
    Loop Behavior=::NewEnumerator1
    Loop Delay=1.1
    Loop Duration Mode=::NewEnumerator1
    UseLoopDelay=false
   )
   &SpawnBurst_Instantaneous(Use Spawn Probability=false, UseLoopCountLimit=false)
   /SetVariables(Emitter.FELight:LinearColor=&MakeCustomLinearColorFromBool(Bool=$User.IsFrontEnd, True Color=<rgba 0.218 -0.3992 1.0836 1>, False Color=<rgba 0 0 0 0>)) {isSetParameters=true}
   /SetVariables(Emitter.ColorGI:LinearColor=&LerpLinearColors(StartColor=<rgba 1.7592 1.1423 2 1>, EndColor=<rgba 0.0074 0.0025 0.0312 1>, LerpFactor=$System.Timeline.Float)) {isSetParameters=true}
  ]
  particleSpawnScript [
   &InitializeParticle(
    Color Mode=::NewEnumerator0
    Lifetime=5
    Mesh Scale=&Multiply_Vector(A=<0.028 0.038 0.035>, B=<0.68 0.68 0.7>)
    Mesh Scale Mode=::NewEnumerator4
    Position Mode=::NewEnumerator1
    Position Offset=<0 -10 -22>
    Position Offset Coordinate Space=::Local
    UsePositionOffset=true
    Write Material Random=true
   )
   &InitialMeshOrientation(
    Mesh Axis To Orient=<0 1 0>
    Mesh Orientation Mode=::NewEnumerator2
    Orientation Coordinate Space=::Local
    Orientation Vector=&NormalizeVector(Vector To Normalize=<1 0 0>, Safe Normalize=true, Safe Fallback Threshold=0, Safe Fallback Vector=<0 0 0.001>)
    Rotation=<0 -0.25 0>
    Use Rotation Vector=true
   )
  ]
  particleUpdateScript [
   &ParticleState(Kill Particles When Lifetime Has Elapsed=false, Let Infinitely Lived Particles Die When Emitter Deactivates=true)
   &SkeletalMeshLocation(Bone Sampling Mode=::NewEnumerator4, MaskSpawnGroup=false, Sampled Position Offset=<0 -3 -1>, Skeletal Mesh=$System.Skeletal Mesh)
   &FN_Timer(Increment Multiplier:NiagaraFloat=&MakeCustomFloatFromBool(
     Bool=$User.IsFrontEnd
     True Float=&FloatFromCurve(
      FloatCurve=DataInterface() {propertyValues {Curve {keys [<at 0 2>,<at 0.4 2>,<at 1.6 0>]}}}
      CurveIndex=&Add_Float(A=$Particles.Age, B=&Multiply_FloatByInt(Float=100, Integer=$User.Deployed))
     )
     False Float=&Multiply_Float(A=&ClampFloat(Float=$System.InterpolateOverTime.InterpolateOverTime_Float.Moving Average), B=1)
    )) {enabled=false}
   &ScaleMeshSize(Scale Factor=&VectorFromFloat(Value=&MakeCustomFloatFromBool(
      Bool=$User.IsFrontEnd
      True Float=&FloatFromCurve(
       FloatCurve=DataInterface() {propertyValues {Curve {keys [{interpMode=RCIM_Cubic, leaveTangent=1.8299, tangentMode=RCTM_Break, time=0, value=0}, {tangentMode=RCTM_User, time=1.5, value=1}]}}}
       CurveIndex=&Add_Float(A=$Particles.Age, B=&Multiply_FloatByInt(Float=100, Integer=$User.Deployed))
       Scale Curve=0.87
      )
      False Float=&FloatFromCurve(
       FloatCurve=DataInterface() {propertyValues {Curve {keys [{interpMode=RCIM_Cubic, leaveTangent=1.8299, tangentMode=RCTM_Break, time=0.4, value=0}, {tangentMode=RCTM_User, time=1.5, value=1}]}}}
       CurveIndex=&Add_Float(A=$Particles.Age, B=&Multiply_FloatByInt(Float=100, Integer=$User.Deployed))
       Scale Curve=0.87
      )
     ))) {enabled=false}
   &ScaleMeshSize(Scale Factor=&MakeVector(
     X=&Add_Float(
      A=1
      B=&Lerp_Float(
       A=&Sine(Normalized Angle=$Emitter.Age, Period=0.25, Scale=0.15, Bias=0)
       B=0
       Alpha=&FloatFromCurve(
        FloatCurve=DataInterface() {propertyValues {Curve {keys [<at 1.5606 1>, {interpMode=RCIM_Cubic, leaveTangent=0.7921, tangentMode=RCTM_Break, time=1.6606, value=0}, <at 3.9572 1>]}}}
        CurveIndex=&MakeCustomFloatFromBool(
         Bool=$User.IsFrontEnd
         True Float=&Add_Float(
          A=&Add_Float(A=$Particles.Age, B=&Multiply_FloatByInt(Float=100, Integer=$User.Deployed))
          B=0.35
         )
         False Float=&Add_Float(A=$Particles.Age, B=&Multiply_FloatByInt(Float=100, Integer=$User.Deployed))
        )
       )
      )
     )
     Y=&Add_Float(
      A=1
      B=&Lerp_Float(
       A=&Cosine(Normalized Angle=$Emitter.Age, Period=0.25, Scale=0.2, Bias=-0.1)
       B=0
       Alpha=&FloatFromCurve(
        FloatCurve=DataInterface() {propertyValues {Curve {keys [<at 1.5606 1>, {interpMode=RCIM_Cubic, leaveTangent=0.7921, tangentMode=RCTM_Break, time=1.6606, value=0}, <at 3.9572 1>]}}}
        CurveIndex=&MakeCustomFloatFromBool(
         Bool=$User.IsFrontEnd
         True Float=&Add_Float(
          A=&Add_Float(A=$Particles.Age, B=&Multiply_FloatByInt(Float=100, Integer=$User.Deployed))
          B=0.35
         )
         False Float=&Add_Float(A=$Particles.Age, B=&Multiply_FloatByInt(Float=100, Integer=$User.Deployed))
        )
       )
      )
     )
     Z=&Add_Float(
      A=1
      B=&Lerp_Float(
       A=&Sine(Normalized Angle=$Emitter.Age, Period=0.25, Scale=0.15, Bias=0)
       B=0
       Alpha=&FloatFromCurve(
        FloatCurve=DataInterface() {propertyValues {Curve {keys [<at 1.5606 1>, {interpMode=RCIM_Cubic, leaveTangent=0.7921, tangentMode=RCTM_Break, time=1.6606, value=0}, <at 3.9572 1>]}}}
        CurveIndex=&MakeCustomFloatFromBool(
         Bool=$User.IsFrontEnd
         True Float=&Add_Float(
          A=&Add_Float(A=$Particles.Age, B=&Multiply_FloatByInt(Float=100, Integer=$User.Deployed))
          B=0.35
         )
         False Float=&Add_Float(A=$Particles.Age, B=&Multiply_FloatByInt(Float=100, Integer=$User.Deployed))
        )
       )
      )
     )
    )) {enabled=false, module=ScaleMeshSize001}
   /SetVariables(Particles.FeetPos:Vector3f=/Niagara/DynamicInputs/Transforms/TransformPosition(
     Position=&Add_Vector(
      A=Linked() {linkedVariable {name=Particles.Position, type=NiagaraPosition}}
      B=&MakeCustomVectorFromBool(Bool=$User.IsFrontEnd, True Vector=<0 0 0>, False Vector=<0 0 0>)
     )
     Source Space=::Local
     Destination Space=::World
    )) {enabled=false, isSetParameters=true}
   /SetVariables(Particles.Feet2Pos:Vector3f=&TransformPosition(
     Position=&Add_Vector(
      A=Linked() {linkedVariable {name=Particles.Position, type=NiagaraPosition}}
      B=&MakeCustomVectorFromBool(Bool=$User.IsFrontEnd, True Vector=<0 0 0>, False Vector=<0 0 0>)
     )
     Source Space=::Local
     Destination Space=::World
    )) {enabled=false, isSetParameters=true}
   &DynamicMaterialParameters(
    Index 0 Param 1=&Multiply_Float(
     A=&Multiply_Float(
      A=&MakeFloatFromVector(Vector=$Particles.Scale, Channel=::NewEnumerator0)
      B=&MakeFloatFromVector(Vector=$Engine.Owner.Scale, Channel=::NewEnumerator0)
     )
     B=2
    )
    Index 0 Param 3=&Power(a=&MakeFloatFromVector(Vector=$Engine.Owner.Scale, Channel=::NewEnumerator0), b=0.5)
    Index 3 Param 1=&FN_FrontEndFloatModifier(
     Gameplay=&FNTimeOfDayScalar(Night Scale=-0.0704, Morning Scale=-0.012, Day Scale=0, Evening Scale=0)
     Front End=-0.024
    )
    Index 3 Param 2=&FNTimeOfDayScalar(Night Scale=-0.228, Morning Scale=0.2, Day Scale=0, Evening Scale=0)
    Index 3 Param 3=&FN_FrontEndFloatModifier(Gameplay=&FNTimeOfDayScalar(Night Scale=0.588, Morning Scale=1, Evening Scale=1), Front End=1.3)
    Module.Param0WriteEnabled=true
    Module.Param0WriteEnabled3=true
    Module.Param1WriteEnabled=false
    Module.Param1WriteEnabled3=true
    Module.Param2WriteEnabled=true
    Module.Param2WriteEnabled3=true
    Module.Param3WriteEnabled=false
    Module.Param3WriteEnabled3=true
    Write Parameter Index 3=true
   )
   &DynamicMaterialParameters(
    Index 2 Param 1=&MakeFloatFromVector(Vector=$Particles.Feet2Pos, Channel=::NewEnumerator0)
    Index 2 Param 2=&MakeFloatFromVector(Vector=$Particles.Feet2Pos, Channel=::NewEnumerator1)
    Index 2 Param 3=&MakeFloatFromVector(Vector=$Particles.Feet2Pos, Channel=::NewEnumerator2)
    Module.Param0WriteEnabled2=false
    Module.Param1WriteEnabled2=false
    Module.Param2WriteEnabled2=false
    Module.Param3WriteEnabled2=false
    Write Parameter Index 0=false
    Write Parameter Index 2=true
   ) {enabled=false, module=DynamicMaterialParameters001}
   /SetVariables(Particles.Position2:NiagaraPosition=&Add_Vector(
     A=Linked() {linkedVariable {name=Particles.Position, type=NiagaraPosition}}
     B=&MakeCustomVectorFromBool(
      Bool=$User.IsFrontEnd
      True Vector=&MakeVector(
       X=0
       Y=0
       Z=&FloatFromCurve(
        FloatCurve=DataInterface() {propertyValues {Curve {keys [<at 0 0>,<at 0.3 0>,<at 0.35 1>,<at 1.45 1.1>,<at 1.75 0>]}}}
        CurveIndex=&Add_Float(A=$Particles.Age, B=&Multiply_FloatByInt(Float=100, Integer=$User.Deployed))
        Scale Curve=20
       )
      )
      False Vector=<0 0 0>
     )
    )) {enabled=false, isSetParameters=true}
  ]
  rendererClasses [&NiagaraMeshRendererProperties]
  renderers [
   {rendererClass=&NiagaraMeshRendererProperties, rendererIndex=0}
   {rendererClass=&NiagaraMeshRendererProperties, rendererIndex=1}
   {rendererClass=&NiagaraMeshRendererProperties, rendererIndex=2}
  ]
 }
}
summary {
 systemName=NS_Backpack_WeaveHarbor_Idle
 emitters [
  {emitterName=MainCloud, rendererClasses [&NiagaraMeshRendererProperties]}
  {emitterName=LightningBase003, rendererClasses [&NiagaraMeshRendererProperties]}
  {emitterName=LandingPuffCloud, rendererClasses [&NiagaraMeshRendererProperties]}
  {emitterName=LightningBase, rendererClasses [&NiagaraMeshRendererProperties]}
  {emitterName=Decals, rendererClasses [&NiagaraSpriteRendererProperties]}
 ]
 userVariables [
  User.bIsFrontEnd {type=NiagaraBool}
  User.bIsFrontEndPreview {type=NiagaraBool}
  User.bWeakPointHit {type=NiagaraBool}
  User.Deployed {type=NiagaraInt32}
  User.HealthPercent {type=NiagaraFloat, defaultValue=1, description=""}
  User.HealthPhase {type=NiagaraFloat, defaultValue=0, description=""}
  User.IsFrontEnd {type=NiagaraBool}
  User.LandBurst {type=NiagaraFloat, defaultValue=0, description=""}
  User.TextFadeStartTime {type=NiagaraFloat, defaultValue=-10, description=""}
 ]
}
