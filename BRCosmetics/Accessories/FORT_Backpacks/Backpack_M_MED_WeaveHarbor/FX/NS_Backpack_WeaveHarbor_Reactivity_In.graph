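# Niagara burst effect for the WeaveHarbor backpack's "Reactivity_In" state, spawning short-lived mesh puff clouds.
# Two near-identical emitters spawn a few puffs on rings of different radius, push them outward, then apply drag and curl noise.
# A bouncy curve pops their size in and shrinks it out.
# Dynamic material parameters follow particle and owner scale and time of day, with an override when User.IsFrontEnd is set.
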NiagaraSystem /BRCosmetics/Accessories/FORT_Backpacks/Backpack_M_MED_WeaveHarbor/FX/NS_Backpack_WeaveHarbor_Reactivity_In
dependencies {
 usedDataInterfaces [&NiagaraDataInterfaceCurve, &NiagaraDataInterfaceSkeletalMesh, &NiagaraDataInterfaceArrayFloat]
 usedDynamicInputs []
 usedModules [
  &self:SystemSpawnScript.SystemScriptSource.SystemScriptGraph.NiagaraNodeAssignment_0.SetVariables_#1
  &SystemState
  &Timeline
 ]
}
emitters {
 LandingPuffCloud {
  emitterUpdateScript [
   &EmitterState()
   &SpawnBurst_Instantaneous(Spawn Count=&RandomRangeInt(Minimum=3, Maximum=4), Use Spawn Probability=true, UseLoopCountLimit=false)
  ]
  particleSpawnScript [
   &InitializeParticle(
    Lifetime Max=0.8
    Lifetime Min=0.4
    Lifetime Mode=::NewEnumerator1
    Mesh Scale=&Multiply_VectorByFloat(Vector=<0.01 0.02 0.02>, Float=&RandomRangeFloat(Minimum=0.6, Maximum=1.1))
    Mesh Scale Mode=::NewEnumerator4
    Sprite Size Mode=::NewEnumerator3
    Uniform Sprite Size=1
    UsePositionOffset=false
    Write Material Random=true
   )
   &ShapeLocation(
    MaskSpawnGroup=false
    Offset=<0 -11 0>
    Offset Mode=::NewEnumerator0
    Override Seed=false
    Ring / Disc Distribution Mode:ENiagaraSphereDistributionMode=::NewEnumerator1
    Ring Radius=4.5
    Rotation Angle=0.25
    Rotation Axis=<0 1 0>
    Rotation Mode=::NewEnumerator1
    Shape Primitive=::NewEnumerator4
    U Position=&ReturnNormalizedExecIndex(
     Normalized Index Scale=&RandomRangeFloat(Minimum=0.9, Maximum=1.25)
     Normalized Index Includes One=false
    )
    UseOwnerScale=true
   ) {module=ShapeLocation001}
   &SkeletalMeshLocation(
    Bone Sampling Mode=::NewEnumerator4
    MaskSpawnGroup=false
    Sampled Position Offset=$Particles.ShapeLocation001.ShapePosition
    Skeletal Mesh=$System.Skeletal Mesh
   ) {enabled=false}
   &AddVelocity(
    Module.Use Falloff Distance=false
    Override Seed=false
    Velocity Mode=::NewEnumerator1
    Velocity Origin:Vector3f=$Particles.ShapeLocation001.ShapePosition
    Velocity Speed=&RandomRangeFloat(Minimum=150, Maximum=250)
   ) {enabled=false}
   &AddVelocity(
    Velocity=$Particles.ShapeLocation001.ShapePosition
    Velocity Speed Scale=&RandomRangeFloat(Minimum=1, Maximum=4)
   ) {module=AddVelocity001}
   &InitialMeshOrientation(Use Rotation Vector=false)
  ]
  particleUpdateScript [
   &Drag(Drag=6, Use Linear Drag=true, Use Rotational Drag=true)
   &CurlNoiseForce(
    Noise Frequency=2
    Noise Quality / Cost=::NewEnumerator0
    Noise Strength=&RandomRangeFloat(Minimum=-250, Maximum=250)
    Pan Noise=false
   )
   &SolveForcesAndVelocity(Clamp Velocity=false, Limit Acceleration=false)
   &ParticleState()
   &DynamicMaterialParameters(
    Index 0 Param 1=&Multiply_Float(
     A=&Multiply_Float(
      A=&MakeFloatFromVector(Vector=$Particles.Scale, Channel=::NewEnumerator0)
      B=&MakeFloatFromVector(Vector=$Engine.Owner.Scale, Channel=::NewEnumerator0)
     )
     B=2
    )
    Index 0 Param 3=&Power(a=&MakeFloatFromVector(Vector=$Engine.Owner.Scale, Channel=::NewEnumerator0), b=0.5)
    Index 3 Param 1=&FNTimeOfDayScalar(Night Scale=-0.0704, Morning Scale=-0.012, Day Scale=0, Evening Scale=0)
    Index 3 Param 2=&FNTimeOfDayScalar(Night Scale=-0.228, Morning Scale=0.2, Day Scale=0, Evening Scale=0)
    Index 3 Param 3=&MakeCustomFloatFromBool(
     Bool=$User.IsFrontEnd
     True Float=1.1
     False Float=&FNTimeOfDayScalar(Night Scale=0.588, Morning Scale=1, Evening Scale=1)
    )
    Module.Param0WriteEnabled=true
    Module.Param0WriteEnabled3=true
    Module.Param1WriteEnabled=false
    Module.Param1WriteEnabled3=true
    Module.Param2WriteEnabled=true
    Module.Param2WriteEnabled3=true
    Module.Param3WriteEnabled=false
    Module.Param3WriteEnabled3=true
    Write Parameter Index 3=true
   )
   &UpdateMeshOrientation(Rotation Rate=&RandomRangeFloat(Minimum=-1, Maximum=1)) {enabled=false}
   &ScaleMeshSize(Scale Factor=&VectorFromFloat(Value=&FloatFromCurve(FloatCurve=DataInterface() {propertyValues {Curve {keys [
          {arriveTangent=16.6204, interpMode=RCIM_Cubic, leaveTangent=16.6204, tangentMode=RCTM_User, time=0, value=0}
          {arriveTangent=0.2654, interpMode=RCIM_Cubic, leaveTangent=0.2654, tangentMode=RCTM_User, time=0.1636, value=1.476}
          {arriveTangent=-0.7988, interpMode=RCIM_Cubic, leaveTangent=-0.7988, time=0.2959, value=0.814}
          {arriveTangent=0.3306, interpMode=RCIM_Cubic, leaveTangent=0.3306, time=0.4349, value=1.2592}
          {arriveTangent=-0.8945, interpMode=RCIM_Cubic, leaveTangent=-0.8945, time=0.5567, value=0.9002}
          {arriveTangent=-0.5833, interpMode=RCIM_Cubic, leaveTangent=-0.5833, time=0.7325, value=0.993}
          {arriveTangent=-3.7129, interpMode=RCIM_Cubic, leaveTangent=-3.7129, time=0.9, value=0.7}
          {arriveTangent=-3.986, interpMode=RCIM_Cubic, leaveTangent=-3.986, tangentMode=RCTM_User, time=1, value=0}
         ]}}})))
  ]
  rendererClasses [&NiagaraMeshRendererProperties]
  renderers [{rendererClass=&NiagaraMeshRendererProperties, rendererIndex=0}]
 }
 LandingPuffCloud001 {
  emitterUpdateScript [
   &EmitterState()
   &SpawnBurst_Instantaneous(Spawn Count=&RandomRangeInt(Minimum=2, Maximum=4), Use Spawn Probability=true, UseLoopCountLimit=false)
  ]
  particleSpawnScript [
   &InitializeParticle(
    Lifetime Max=0.8
    Lifetime Min=0.45
    Lifetime Mode=::NewEnumerator1
    Mesh Scale=&Multiply_VectorByFloat(Vector=<0.01 0.02 0.02>, Float=&RandomRangeFloat(Minimum=0.6, Maximum=1.1))
    Mesh Scale Mode=::NewEnumerator4
    Sprite Size Mode=::NewEnumerator3
    Uniform Sprite Size=1
    UsePositionOffset=false
    Write Material Random=true
   )
   &SkeletalMeshLocation(Bone Sampling Mode=::NewEnumerator4, MaskSpawnGroup=false, Skeletal Mesh=$System.Skeletal Mesh) {enabled=false}
   &ShapeLocation(
    MaskSpawnGroup=false
    Offset=<0 -5 0>
    Offset Mode=::NewEnumerator0
    Override Seed=false
    Ring / Disc Distribution Mode:ENiagaraSphereDistributionMode=::NewEnumerator1
    Ring Radius=9
    Rotation Angle=0.25
    Rotation Axis=<0 1 0>
    Rotation Mode=::NewEnumerator1
    Shape Primitive=::NewEnumerator4
    U Position=&ReturnNormalizedExecIndex(
     Normalized Index Scale=&RandomRangeFloat(Minimum=0.9, Maximum=1.25)
     Normalized Index Includes One=false
    )
    UseOwnerScale=true
   ) {module=ShapeLocation001}
   &AddVelocity(
    Module.Use Falloff Distance=false
    Override Seed=false
    Velocity Mode=::NewEnumerator1
    Velocity Origin:Vector3f=$Particles.ShapeLocation001.ShapePosition
    Velocity Speed=&RandomRangeFloat(Minimum=150, Maximum=250)
   ) {enabled=false}
   &AddVelocity(
    Velocity=$Particles.ShapeLocation001.ShapePosition
    Velocity Speed Scale=&RandomRangeFloat(Minimum=1, Maximum=4)
   ) {module=AddVelocity001}
   &InitialMeshOrientation(Use Rotation Vector=false)
  ]
  particleUpdateScript [
   &Drag(Drag=6, Use Linear Drag=true, Use Rotational Drag=true)
   &CurlNoiseForce(
    Noise Frequency=2
    Noise Quality / Cost=::NewEnumerator0
    Noise Strength=&RandomRangeFloat(Minimum=-200, Maximum=200)
    Pan Noise=false
   )
   &SolveForcesAndVelocity(Clamp Velocity=false, Limit Acceleration=false)
   &ParticleState()
   &DynamicMaterialParameters(
    Index 0 Param 1=&Multiply_Float(
     A=&Multiply_Float(
      A=&MakeFloatFromVector(Vector=$Particles.Scale, Channel=::NewEnumerator0)
      B=&MakeFloatFromVector(Vector=$Engine.Owner.Scale, Channel=::NewEnumerator0)
     )
     B=2
    )
    Index 0 Param 3=&Power(a=&MakeFloatFromVector(Vector=$Engine.Owner.Scale, Channel=::NewEnumerator0), b=0.5)
    Index 3 Param 1=&FNTimeOfDayScalar(Night Scale=-0.0704, Morning Scale=-0.012, Day Scale=0, Evening Scale=0)
    Index 3 Param 2=&FNTimeOfDayScalar(Night Scale=-0.228, Morning Scale=0.2, Day Scale=0, Evening Scale=0)
    Index 3 Param 3=&MakeCustomFloatFromBool(
     Bool=$User.IsFrontEnd
     True Float=1.1
     False Float=&FNTimeOfDayScalar(Night Scale=0.588, Morning Scale=1, Evening Scale=1)
    )
    Module.Param0WriteEnabled=true
    Module.Param0WriteEnabled3=true
    Module.Param1WriteEnabled=false
    Module.Param1WriteEnabled3=true
    Module.Param2WriteEnabled=true
    Module.Param2WriteEnabled3=true
    Module.Param3WriteEnabled=false
    Module.Param3WriteEnabled3=true
    Write Parameter Index 3=true
   )
   &UpdateMeshOrientation(Rotation Rate=&RandomRangeFloat(Minimum=-1, Maximum=1)) {enabled=false}
   &ScaleMeshSize(Scale Factor=&VectorFromFloat(Value=&FloatFromCurve(FloatCurve=DataInterface() {propertyValues {Curve {keys [
          {arriveTangent=16.6204, interpMode=RCIM_Cubic, leaveTangent=16.6204, tangentMode=RCTM_User, time=0, value=0}
          {arriveTangent=0.2654, interpMode=RCIM_Cubic, leaveTangent=0.2654, tangentMode=RCTM_User, time=0.1636, value=1.476}
          {arriveTangent=-0.7988, interpMode=RCIM_Cubic, leaveTangent=-0.7988, time=0.2959, value=0.814}
          {arriveTangent=0.3306, interpMode=RCIM_Cubic, leaveTangent=0.3306, time=0.4349, value=1.2592}
          {arriveTangent=-0.8945, interpMode=RCIM_Cubic, leaveTangent=-0.8945, time=0.5567, value=0.9002}
          {arriveTangent=-0.5833, interpMode=RCIM_Cubic, leaveTangent=-0.5833, time=0.7325, value=0.993}
          {arriveTangent=-3.7129, interpMode=RCIM_Cubic, leaveTangent=-3.7129, time=0.9, value=0.7}
          {arriveTangent=-3.986, interpMode=RCIM_Cubic, leaveTangent=-3.986, tangentMode=RCTM_User, time=1, value=0}
         ]}}})))
  ]
  rendererClasses [&NiagaraMeshRendererProperties]
  renderers [{rendererClass=&NiagaraMeshRendererProperties, rendererIndex=0}]
 }
}
summary {
 systemName=NS_Backpack_WeaveHarbor_Reactivity_In
 emitters [
  {emitterName=LandingPuffCloud, rendererClasses [&NiagaraMeshRendererProperties]}
  {emitterName=LandingPuffCloud001, rendererClasses [&NiagaraMeshRendererProperties]}
 ]
 userVariables [
  User.bIsFrontEnd {type=NiagaraBool}
  User.bIsFrontEndPreview {type=NiagaraBool}
  User.bWeakPointHit {type=NiagaraBool}
  User.Deployed {type=NiagaraInt32}
  User.HealthPercent {type=NiagaraFloat, defaultValue=1, description=""}
  User.IsFrontEnd {type=NiagaraBool}
  User.LandBurst {type=NiagaraFloat, defaultValue=0, description=""}
 ]
}
