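# Niagara burst effect for the WeaveHarbor backpack cosmetic's "Reactivity Out" event, combining smoke puffs, lightning, sparks and a flash.
# Two mesh puff emitters push curl-noise-driven rings outward, while red or purple lightning meshes and stretched sparks radiate from offset spheres.
# A short flickering red sprite follows the skeletal mesh.
# Front-end and time-of-day inputs adjust brightness, size and materials; the health-driven lightning spawn rate is disabled.
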NiagaraSystem /BRCosmetics/Accessories/FORT_Backpacks/Backpack_M_MED_WeaveHarbor/FX/NS_Backpack_WeaveHarbor_Reactivity_Out
dependencies {
 usedDataInterfaces [
  &NiagaraDataInterfaceCurve
  &NiagaraDataInterfaceArrayColor
  &NiagaraDataInterfaceMeshRendererInfo
  &NiagaraDataInterfaceColorCurve
  &NiagaraDataInterfaceSkeletalMesh
  &NiagaraDataInterfaceArrayFloat
 ]
 usedDynamicInputs []
 usedModules [
  &self:SystemSpawnScript.SystemScriptSource.SystemScriptGraph.NiagaraNodeAssignment_0.SetVariables_#1
  &SystemState
  &Timeline
 ]
}
emitters {
 Decals {
  emitterUpdateScript [
   &EmitterState()
   &SpawnBurst_Instantaneous(Spawn Count=1, Use Spawn Probability=false) {module=SpawnBurst_Instantaneous001}
  ]
  particleSpawnScript [&InitializeParticle(
    Color=<rgba 0.5625 0.0103 0 0.4>
    Lifetime=0.45
    Sprite Size Mode=::NewEnumerator3
    Uniform Sprite Size=38
    UsePositionOffset=false
    Write Material Random=true
   ) {module=InitializeParticle001}]
  particleUpdateScript [
   &ParticleState()
   &SkeletalMeshLocation(Bone Sampling Mode=::NewEnumerator4, MaskSpawnGroup=false, Sampled Position Offset=<0 -6 -1>, Skeletal Mesh=$System.Skeletal Mesh)
   &Drag(Drag=10, Use Linear Drag=true, Use Rotational Drag=false)
   &SolveForcesAndVelocity(Clamp Velocity=false, Limit Acceleration=false)
   &CameraOffset(Camera Offset Amount=15)
   &DynamicMaterialParameters(
    Index 0 Param 1=2
    Index 0 Param 2=1
    Index 0 Param 3=-0.5
    Module.Param0WriteEnabled=true
    Module.Param1WriteEnabled=true
    Module.Param2WriteEnabled=true
    Module.Param3WriteEnabled=true
   )
   &ScaleColor(
    Scale Alpha=&FloatFromCurve(FloatCurve=DataInterface() {propertyValues {Curve {keys [
         <at 0 0 RCIM_Cubic>
         {arriveTangent=0.7215, interpMode=RCIM_Cubic, leaveTangent=0.7215, tangentMode=RCTM_User, time=0.2, value=0.8}
         {arriveTangent=0.0442, interpMode=RCIM_Cubic, leaveTangent=-1.1429, tangentMode=RCTM_Break, time=0.3261, value=0.1993}
         {arriveTangent=0.3826, interpMode=RCIM_Cubic, leaveTangent=0.3826, time=0.5, value=0.75}
         {arriveTangent=-0.175, interpMode=RCIM_Cubic, leaveTangent=-0.175, time=0.6404, value=0.3195}
         <at 0.7857 0.7>
         <at 1 0>
        ]}}})
    Scale RGB=&Multiply_VectorByFloat(Vector=<1 1 1>, Float=&FN_FrontEndFloatModifier(Gameplay=1, Front End=0.5))
    ScaleA=true
    ScaleRGB=true
   )
  ]
  rendererClasses [&NiagaraSpriteRendererProperties]
  renderers [{rendererClass=&NiagaraSpriteRendererProperties}]
 }
 LandingPuffCloud {
  emitterUpdateScript [
   &EmitterState()
   &SpawnBurst_Instantaneous(Spawn Count=&RandomRangeInt(Minimum=3, Maximum=5), Use Spawn Probability=true, UseLoopCountLimit=false)
  ]
  particleSpawnScript [
   &InitializeParticle(
    Lifetime Max=0.8
    Lifetime Min=0.4
    Lifetime Mode=::NewEnumerator1
    Mesh Scale=&Multiply_VectorByFloat(Vector=<0.01 0.02 0.02>, Float=&RandomRangeFloat(Minimum=0.6, Maximum=1.1))
    Mesh Scale Mode=::NewEnumerator4
    Sprite Size Mode=::NewEnumerator3
    Uniform Sprite Size=1
    UsePositionOffset=false
    Write Material Random=true
   )
   &ShapeLocation(
    MaskSpawnGroup=false
    Offset=<0 -11 0>
    Offset Mode=::NewEnumerator0
    Override Seed=false
    Ring / Disc Distribution Mode:ENiagaraSphereDistributionMode=::NewEnumerator1
    Ring Radius=4.5
    Rotation Angle=0.25
    Rotation Axis=<0 1 0>
    Rotation Mode=::NewEnumerator1
    Shape Primitive=::NewEnumerator4
    U Position=&ReturnNormalizedExecIndex(
     Normalized Index Scale=&RandomRangeFloat(Minimum=0.9, Maximum=1.25)
     Normalized Index Includes One=false
    )
    UseOwnerScale=true
   ) {module=ShapeLocation001}
   &SkeletalMeshLocation(
    Bone Sampling Mode=::NewEnumerator4
    MaskSpawnGroup=false
    Sampled Position Offset=$Particles.ShapeLocation001.ShapePosition
    Skeletal Mesh=$System.Skeletal Mesh
   ) {enabled=false}
   &AddVelocity(
    Module.Use Falloff Distance=false
    Override Seed=false
    Velocity Mode=::NewEnumerator1
    Velocity Origin:Vector3f=$Particles.ShapeLocation001.ShapePosition
    Velocity Speed=&RandomRangeFloat(Minimum=150, Maximum=250)
   ) {enabled=false}
   &AddVelocity(
    Velocity=$Particles.ShapeLocation001.ShapePosition
    Velocity Speed Scale=&RandomRangeFloat(Minimum=2, Maximum=6)
   ) {module=AddVelocity001}
   &InitialMeshOrientation(Use Rotation Vector=false)
  ]
  particleUpdateScript [
   &Drag(Drag=8, Use Linear Drag=true, Use Rotational Drag=true)
   &CurlNoiseForce(
    Noise Frequency=2
    Noise Quality / Cost=::NewEnumerator0
    Noise Strength=&RandomRangeFloat(Minimum=-500, Maximum=500)
    Pan Noise=false
   )
   &SolveForcesAndVelocity(Clamp Velocity=false, Limit Acceleration=false)
   &ParticleState()
   &DynamicMaterialParameters(
    Index 0 Param 1=&Multiply_Float(
     A=&Multiply_Float(
      A=&MakeFloatFromVector(Vector=$Particles.Scale, Channel=::NewEnumerator0)
      B=&MakeFloatFromVector(Vector=$Engine.Owner.Scale, Channel=::NewEnumerator0)
     )
     B=2
    )
    Index 0 Param 3=&Power(a=&MakeFloatFromVector(Vector=$Engine.Owner.Scale, Channel=::NewEnumerator0), b=0.5)
    Index 3 Param 1=&FNTimeOfDayScalar(Night Scale=-0.0704, Morning Scale=-0.012, Day Scale=0, Evening Scale=0)
    Index 3 Param 2=&FNTimeOfDayScalar(Night Scale=-0.228, Morning Scale=0.2, Day Scale=0, Evening Scale=0)
    Index 3 Param 3=&MakeCustomFloatFromBool(
     Bool=$User.IsFrontEnd
     True Float=1.1
     False Float=&FNTimeOfDayScalar(Night Scale=0.588, Morning Scale=1, Evening Scale=1)
    )
    Module.Param0WriteEnabled=true
    Module.Param0WriteEnabled3=true
    Module.Param1WriteEnabled=false
    Module.Param1WriteEnabled3=true
    Module.Param2WriteEnabled=true
    Module.Param2WriteEnabled3=true
    Module.Param3WriteEnabled=false
    Module.Param3WriteEnabled3=true
    Write Parameter Index 3=true
   )
   &UpdateMeshOrientation(Rotation Rate=&RandomRangeFloat(Minimum=-1, Maximum=1)) {enabled=false}
   &ScaleMeshSize(Scale Factor=&VectorFromFloat(Value=&FloatFromCurve(FloatCurve=DataInterface() {propertyValues {Curve {keys [
          {arriveTangent=16.6204, interpMode=RCIM_Cubic, leaveTangent=16.6204, tangentMode=RCTM_User, time=0, value=0}
          {arriveTangent=0.2654, interpMode=RCIM_Cubic, leaveTangent=0.2654, tangentMode=RCTM_User, time=0.1636, value=1.476}
          {arriveTangent=-0.7988, interpMode=RCIM_Cubic, leaveTangent=-0.7988, time=0.2959, value=0.814}
          {arriveTangent=0.3306, interpMode=RCIM_Cubic, leaveTangent=0.3306, time=0.4349, value=1.2592}
          {arriveTangent=-0.8945, interpMode=RCIM_Cubic, leaveTangent=-0.8945, time=0.5567, value=0.9002}
          {arriveTangent=-0.5833, interpMode=RCIM_Cubic, leaveTangent=-0.5833, time=0.7325, value=0.993}
          {arriveTangent=-3.7129, interpMode=RCIM_Cubic, leaveTangent=-3.7129, time=0.9, value=0.7}
          {arriveTangent=-3.986, interpMode=RCIM_Cubic, leaveTangent=-3.986, tangentMode=RCTM_User, time=1, value=0}
         ]}}})))
  ]
  rendererClasses [&NiagaraMeshRendererProperties]
  renderers [{rendererClass=&NiagaraMeshRendererProperties, rendererIndex=0}]
 }
 LandingPuffCloud001 {
  emitterUpdateScript [
   &EmitterState()
   &SpawnBurst_Instantaneous(Spawn Count=&RandomRangeInt(Minimum=2, Maximum=4), Use Spawn Probability=true, UseLoopCountLimit=false)
  ]
  particleSpawnScript [
   &InitializeParticle(
    Lifetime Max=0.8
    Lifetime Min=0.45
    Lifetime Mode=::NewEnumerator1
    Mesh Scale=&Multiply_VectorByFloat(Vector=<0.01 0.02 0.02>, Float=&RandomRangeFloat(Minimum=0.6, Maximum=1.1))
    Mesh Scale Mode=::NewEnumerator4
    Sprite Size Mode=::NewEnumerator3
    Uniform Sprite Size=1
    UsePositionOffset=false
    Write Material Random=true
   )
   &SkeletalMeshLocation(Bone Sampling Mode=::NewEnumerator4, MaskSpawnGroup=false, Skeletal Mesh=$System.Skeletal Mesh) {enabled=false}
   &ShapeLocation(
    MaskSpawnGroup=false
    Offset=<0 -5 0>
    Offset Mode=::NewEnumerator0
    Override Seed=false
    Ring / Disc Distribution Mode:ENiagaraSphereDistributionMode=::NewEnumerator1
    Ring Radius=12.6
    Rotation Angle=0.25
    Rotation Axis=<0 1 0>
    Rotation Mode=::NewEnumerator1
    Shape Primitive=::NewEnumerator4
    U Position=&ReturnNormalizedExecIndex(
     Normalized Index Scale=&RandomRangeFloat(Minimum=0.9, Maximum=1.25)
     Normalized Index Includes One=false
    )
    UseOwnerScale=true
   ) {module=ShapeLocation001}
   &AddVelocity(
    Module.Use Falloff Distance=false
    Override Seed=false
    Velocity Mode=::NewEnumerator1
    Velocity Origin:Vector3f=$Particles.ShapeLocation001.ShapePosition
    Velocity Speed=&RandomRangeFloat(Minimum=150, Maximum=250)
   ) {enabled=false}
   &AddVelocity(
    Velocity=$Particles.ShapeLocation001.ShapePosition
    Velocity Speed Scale=&RandomRangeFloat(Minimum=1, Maximum=4)
   ) {module=AddVelocity001}
   &InitialMeshOrientation(Use Rotation Vector=false)
  ]
  particleUpdateScript [
   &Drag(Drag=8, Use Linear Drag=true, Use Rotational Drag=true)
   &CurlNoiseForce(
    Noise Frequency=2
    Noise Quality / Cost=::NewEnumerator0
    Noise Strength=&RandomRangeFloat(Minimum=-450, Maximum=450)
    Pan Noise=false
   )
   &SolveForcesAndVelocity(Clamp Velocity=false, Limit Acceleration=false)
   &ParticleState()
   &DynamicMaterialParameters(
    Index 0 Param 1=&Multiply_Float(
     A=&Multiply_Float(
      A=&MakeFloatFromVector(Vector=$Particles.Scale, Channel=::NewEnumerator0)
      B=&MakeFloatFromVector(Vector=$Engine.Owner.Scale, Channel=::NewEnumerator0)
     )
     B=2
    )
    Index 0 Param 3=&Power(a=&MakeFloatFromVector(Vector=$Engine.Owner.Scale, Channel=::NewEnumerator0), b=0.5)
    Index 3 Param 1=&FNTimeOfDayScalar(Night Scale=-0.0704, Morning Scale=-0.012, Day Scale=0, Evening Scale=0)
    Index 3 Param 2=&FNTimeOfDayScalar(Night Scale=-0.228, Morning Scale=0.2, Day Scale=0, Evening Scale=0)
    Index 3 Param 3=&MakeCustomFloatFromBool(
     Bool=$User.IsFrontEnd
     True Float=1.1
     False Float=&FNTimeOfDayScalar(Night Scale=0.588, Morning Scale=1, Evening Scale=1)
    )
    Module.Param0WriteEnabled=true
    Module.Param0WriteEnabled3=true
    Module.Param1WriteEnabled=false
    Module.Param1WriteEnabled3=true
    Module.Param2WriteEnabled=true
    Module.Param2WriteEnabled3=true
    Module.Param3WriteEnabled=false
    Module.Param3WriteEnabled3=true
    Write Parameter Index 3=true
   )
   &UpdateMeshOrientation(Rotation Rate=&RandomRangeFloat(Minimum=-1, Maximum=1)) {enabled=false}
   &ScaleMeshSize(Scale Factor=&VectorFromFloat(Value=&FloatFromCurve(FloatCurve=DataInterface() {propertyValues {Curve {keys [
          {arriveTangent=16.6204, interpMode=RCIM_Cubic, leaveTangent=16.6204, tangentMode=RCTM_User, time=0, value=0}
          {arriveTangent=0.2654, interpMode=RCIM_Cubic, leaveTangent=0.2654, tangentMode=RCTM_User, time=0.1636, value=1.476}
          {arriveTangent=-0.7988, interpMode=RCIM_Cubic, leaveTangent=-0.7988, time=0.2959, value=0.814}
          {arriveTangent=0.3306, interpMode=RCIM_Cubic, leaveTangent=0.3306, time=0.4349, value=1.2592}
          {arriveTangent=-0.8945, interpMode=RCIM_Cubic, leaveTangent=-0.8945, time=0.5567, value=0.9002}
          {arriveTangent=-0.5833, interpMode=RCIM_Cubic, leaveTangent=-0.5833, time=0.7325, value=0.993}
          {arriveTangent=-3.7129, interpMode=RCIM_Cubic, leaveTangent=-3.7129, time=0.9, value=0.7}
          {arriveTangent=-3.986, interpMode=RCIM_Cubic, leaveTangent=-3.986, tangentMode=RCTM_User, time=1, value=0}
         ]}}})))
  ]
  rendererClasses [&NiagaraMeshRendererProperties]
  renderers [{rendererClass=&NiagaraMeshRendererProperties, rendererIndex=0}]
 }
 LightningBase {
  emitterUpdateScript [
   &EmitterState()
   &SpawnBurst_Instantaneous(Spawn Count=8, Use Spawn Probability=false, UseLoopCountLimit=false)
   &SpawnRate(
    SpawnRate=&Multiply_Float(
     A=3
     B=&MakeCustomFloatFromBool(
      Bool=&SetBoolByFloatComparison(A=$User.HealthPercent, B=0.51, Comparison Type=::NewEnumerator2)
      True Float=0
      False Float=&Lerp_Float(
       A=0.3
       B=1.1
       Alpha=&RemapRange(Input Value To Remap=$User.HealthPercent, Input Min=0.5, Input Max=0.25, Output Min=0, Output Max=1)
      )
     )
    )
    Use Spawn Probability=true
   ) {enabled=false}
  ]
  particleSpawnScript [
   &InitializeParticle(
    Color=&SelectLinearColorFromArray(
     Array Sampling Mode=::NewEnumerator0
     Direct Array Index=?
     Color Selection Array=DataInterface() {propertyValues {ColorData [<rgba 0.6875 0.0058 0.0469 1>, <rgba 0.0995 0.0264 0.526 1>]}}
     Random Seed=0
    )
    Lifetime Max=0.8
    Lifetime Min=0.4
    Lifetime Mode=::NewEnumerator1
    Mesh Renderer Array Visibility Mode=::NewEnumerator3
    Mesh Renderer Info=DataInterface() {propertyValues {MeshRenderer=&self:LightningBase_1.NiagaraMeshRendererProperties_0}}
    Mesh Scale=<0.036 0.036 0.27>
    Mesh Scale Mode=::NewEnumerator4
    UsePositionOffset=false
    Write Material Random=true
   )
   &ShapeLocation(
    MaskSpawnGroup=false
    Offset=<0 -6 0>
    Offset Mode=::NewEnumerator0
    Shape Origin=$Engine.Emitter.SimulationPosition
    Sphere Distribution=::NewEnumerator1
    Sphere Radius=1
    Sphere Radius Position:NiagaraFloat=&RandomRangeFloat(Minimum=-1, Maximum=1)
    UseOwnerScale=true
    V Position=&ReturnNormalizedExecIndex()
   )
   &InitialMeshOrientation(
    Mesh Orientation Mode=::NewEnumerator2
    Orientation Vector=$Particles.ShapeLocation.ShapeNormal
    Rotation=&RandomRangeVector(Minimum=<0 0 0>, Maximum=<0 0 1>)
    Use Rotation Vector=true
   )
   &CameraOffset(Camera Offset Amount=4)
  ]
  particleUpdateScript [
   &SkeletalMeshLocation(
    MaskSpawnGroup=false
    Mesh Sampling Type=::NewEnumerator5
    Sampled Position Offset=<0 -5 -120>
    Skeletal Mesh=$System.Skeletal Mesh
    Socket Sampling Mode=::NewEnumerator5
   ) {enabled=false}
   &SolveForcesAndVelocity(Clamp Velocity=false, Limit Acceleration=false)
   &ParticleState()
   &ScaleColor(
    Scale Alpha=&FloatFromCurve(FloatCurve=DataInterface() {propertyValues {Curve {keys [
         {interpMode=RCIM_Cubic, leaveTangent=0.2357, tangentMode=RCTM_Break, time=0.3, value=1}
         {arriveTangent=-1.6327, tangentMode=RCTM_User, time=1, value=0}
        ]}}})
    Scale RGB=&VectorFromFloat(Value=&FloatFromCurve(FloatCurve=DataInterface() {propertyValues {Curve {keys [
          <at 0 1>
          {arriveTangent=1.25, leaveTangent=1.25, time=0.3, value=1}
          {interpMode=RCIM_Cubic, leaveTangent=-5.9679, tangentMode=RCTM_Break, time=0.4, value=1.5}
          <at 1 0.4>
         ]}}}))
    ScaleA=true
    ScaleRGB=true
   )
   &DynamicMaterialParameters(
    Index 0 Param 1=&FloatFromCurve(FloatCurve=DataInterface() {propertyValues {Curve {keys [{interpMode=RCIM_Cubic, tangentMode=RCTM_User, time=0, value=0.25}, {interpMode=RCIM_Cubic, tangentMode=RCTM_User, time=1, value=1}]}}})
    Index 0 Param 2=&RandomRangeFloat(Minimum=4, Maximum=8)
    Index 0 Param 3=&FloatFromCurve(
     FloatCurve=DataInterface() {propertyValues {Curve {keys [
         <at 0 1>
         {interpMode=RCIM_Cubic, leaveTangent=0.1085, tangentMode=RCTM_Break, time=0.1038, value=1}
         {interpMode=RCIM_Cubic, leaveTangent=-8.5033, tangentMode=RCTM_Break, time=0.2038, value=2.625}
         <at 1 0>
        ]}}}
     Scale Curve=&RandomRangeFloat(Minimum=15, Maximum=20)
    )
    Module.Index0Param0Write=true
    Module.Index0Param1Write=true
    Module.Index0Param2Write=true
    Module.Index0Param3Write=true
   )
   /SetVariables(Particles.Dy2:Vector4f=&MakeVector4(
     X=&FloatFromCurve(FloatCurve=DataInterface() {propertyValues {Curve {keys [{interpMode=RCIM_Cubic, tangentMode=RCTM_User, time=0, value=0}, {interpMode=RCIM_Cubic, tangentMode=RCTM_User, time=1, value=1}]}}})
     Y=&RandomRangeFloat(Minimum=4, Maximum=12)
     Z=&FloatFromCurve(
      FloatCurve=DataInterface() {propertyValues {Curve {keys [
          <at 0 1>
          {interpMode=RCIM_Cubic, leaveTangent=0.1085, tangentMode=RCTM_Break, time=0.1038, value=1}
          {interpMode=RCIM_Cubic, leaveTangent=-8.5033, tangentMode=RCTM_Break, time=0.2038, value=2.625}
          <at 1 0>
         ]}}}
      Scale Curve=&RandomRangeFloat(Minimum=15, Maximum=20)
     )
     W=0
    )) {isSetParameters=true}
  ]
  rendererClasses [&NiagaraMeshRendererProperties]
  renderers [
   {rendererClass=&NiagaraMeshRendererProperties, rendererIndex=0}
   {rendererClass=&NiagaraMeshRendererProperties, rendererIndex=1}
  ]
 }
 Spark {
  emitterUpdateScript [
   &EmitterState()
   &SpawnBurst_Instantaneous(Spawn Count=8, Use Spawn Probability=false)
  ]
  particleSpawnScript [
   &InitializeParticle(
    Color=<rgba 1 0.1442 0.194 1>
    Lifetime Max=0.4
    Lifetime Min=0.1
    Lifetime Mode=::NewEnumerator1
    Position Offset=<0 -5 0>
    Sprite Size Max=<2.5 8>
    Sprite Size Min=<2 5>
    Sprite Size Mode=::NewEnumerator2
    UsePositionOffset=true
    Write Material Random=true
   )
   &ShapeLocation(
    Hemisphere Distribution=<0.5 1>
    MaskSpawnGroup=false
    Override Seed=false
    Rotation Mode=::NewEnumerator2
    Sphere Radius=10
    UseOwnerScale=true
    Yaw / Pitch / Roll=<0 0 0.5>
   )
   &AddVelocityFromPoint(
    Module.Use Falloff Distance=false
    Origin Offset=<0 -5 0>
    Velocity Strength=&RandomRangeFloat(Minimum=150, Maximum=200)
   )
  ]
  particleUpdateScript [
   &ParticleState()
   &ScaleSpriteSize(
    Uniform Curve Scale=&FN_FrontEndFloatModifier(Gameplay=1, Front End=0.2)
    Uniform Curve Sprite Scale=DataInterface() {propertyValues {Curve {keys [{interpMode=RCIM_Cubic, tangentMode=RCTM_User, time=0, value=1}, {interpMode=RCIM_Cubic, tangentMode=RCTM_User, time=1, value=0}]}}}
   )
   &ScaleColor(
    Linear Color Curve=DataInterface() {propertyValues {AlphaCurve {keys [<at 0.0074 1>,<at 0.8905 1>]}, BlueCurve {keys [<at 0 15>]}, GreenCurve {keys [<at 0 15>]}, RedCurve {keys [<at 0 15>]}}}
    Scale Mode=::NewEnumerator2
   )
   &ScaleColor(
    Scale Alpha=&FNTimeOfDayScalar(Night Scale=0.5, Morning Scale=0.3, Evening Scale=0.3)
    Scale RGB=&VectorFromFloat(Value=&Multiply_Float(A=10, B=1))
    ScaleA=true
    ScaleRGB=true
   ) {module=ScaleColor001}
   &Drag(
    Drag=&FloatFromCurve(FloatCurve=DataInterface() {propertyValues {Curve {keys [{interpMode=RCIM_Cubic, leaveTangent=6.8554, tangentMode=RCTM_Break, time=0, value=1}, <at 1 6>]}}})
    Use Linear Drag=true
    Use Rotational Drag=false
   )
   &CurlNoiseForce(Noise Strength=&RandomRangeFloat(Minimum=200, Maximum=400), Pan Noise=true)
   &GravityForce(Gravity=&Multiply_VectorByFloat(
     Vector=&RandomRangeVector(Minimum=<0 0 -100>, Maximum=<0 0 -250>)
     Float=&FN_FrontEndFloatModifier(Gameplay=1, Front End=1)
    ))
   &SolveForcesAndVelocity(Clamp Velocity=false, Limit Acceleration=false)
   &ScaleSpriteSizeBySpeed(Max Scale Factor=<1 5>, Min Scale Factor=<1 0.05>)
  ]
  rendererClasses [&NiagaraSpriteRendererProperties]
  renderers [{rendererClass=&NiagaraSpriteRendererProperties}]
 }
}
summary {
 systemName=NS_Backpack_WeaveHarbor_Reactivity_Out
 emitters [
  {emitterName=LandingPuffCloud, rendererClasses [&NiagaraMeshRendererProperties]}
  {emitterName=LightningBase, rendererClasses [&NiagaraMeshRendererProperties]}
  {emitterName=LandingPuffCloud001, rendererClasses [&NiagaraMeshRendererProperties]}
  {emitterName=Spark, rendererClasses [&NiagaraSpriteRendererProperties]}
  {emitterName=Decals, rendererClasses [&NiagaraSpriteRendererProperties]}
 ]
 userVariables [
  User.bIsFrontEnd {type=NiagaraBool}
  User.bIsFrontEndPreview {type=NiagaraBool}
  User.bWeakPointHit {type=NiagaraBool}
  User.Deployed {type=NiagaraInt32}
  User.HealthPercent {type=NiagaraFloat, defaultValue=1, description=""}
  User.IsFrontEnd {type=NiagaraBool}
  User.LandBurst {type=NiagaraFloat, defaultValue=0, description=""}
 ]
}
