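# Translucent two-sided parent material for the WeaveHarbor backpack cosmetic, drawing layered text, shadow and boards over the character parent functions.
# BumpOffset parallax separates three phase-swapped texture masks, with front/back height flipping and view-angle-based virtual thickness shaping opacity.
# Timestamped in/out transitions sample curve atlases to drive offset, scale and twist in World Position Offset.
# A camera offset, inverted when viewed from behind, is added on top.
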Material /BRCosmetics/Accessories/FORT_Backpacks/Backpack_M_MED_WeaveHarbor/Materials/M_Backpack_WeaveHarbor_Parent
graph {
 properties {
  bHasPixelAnimation=true
  bOutputTranslucentVelocity=true
  bUseMaterialAttributes=true
  bUsedWithMeshDeformer=true
  bUsedWithSkeletalMesh=true
  blendMode=BLEND_Translucent
  pixelDepthOffsetMode=PDOM_Legacy
  shadingModel=MSM_FromMaterialExpression
  twoSided=true
  usedShadingModels=Default Lit | Subsurface | Clear Coat | Subsurface Profile | Cloth
 }
 nodes [
  Add()
  Frac(None=Time())
  n1 = MaterialFunctionCall() {materialFunction=&MF_CharacterParent_2023_PreFX}
  n2 = Multiply(A=NamedRerouteUsage(), B=VectorParameter() {group=WeaveHarbor, parameterName=TextColor, sortPriority=100, defaultValue=<rgba 1 1 1 1>}.RGB)
  n3 = VectorParameter() {group=WeaveHarbor, parameterName=ShadowColor, sortPriority=100, defaultValue=<rgba 0.0685 0.0116 0.1356 1>}
  n4 = Multiply(A=NamedRerouteUsage(), B=@n3.RGB)
  n5 = NamedRerouteUsage()
  n6 = VectorParameter() {group=WeaveHarbor, parameterName=BoardsShadowColor, sortPriority=100, defaultValue=<rgba 0.5271 0.0003 0.0222 1>}
  n7 = VectorParameter() {group=WeaveHarbor, parameterName=BoardsColor, sortPriority=100, defaultValue=<rgba 0.956 0.0232 0.0222 1>}
  n8 = NamedRerouteUsage()
  n9 = LinearInterpolate(A=@n6.RGB, B=@n7.RGB, Alpha=Saturate(None=ConstantBiasScale(None=@n8) {bias=-0.8, desc=to separate board from shaded edge, scale=5}))
  n10 = NamedRerouteUsage()
  n11 = LinearInterpolate(
   A=LinearInterpolate(A=LinearInterpolate(A=@n2, B=@n4, Alpha=@n5), B=@n9, Alpha=NamedRerouteUsage())
   B=LinearInterpolate(A=LinearInterpolate(A=@n9, B=@n4, Alpha=@n10), B=@n2, Alpha=NamedRerouteUsage())
   Alpha=NamedRerouteUsage()
  )
  n12 = NamedRerouteUsage()
  n13 = NamedRerouteUsage()
  n14 = ScalarParameter() {defaultValue=0.5, group=WeaveHarbor, parameterName=Height_03_Boards, sortPriority=80}
  n15 = LinearInterpolate(A=OneMinus(None=@n14), B=@n14, Alpha=NamedRerouteUsage())
  n16 = Multiply(A=NamedRerouteUsage(), B=@n15)
  n17 = ScalarParameter() {defaultValue=0.7, group=WeaveHarbor, parameterName=Height_02_TextShadow, sortPriority=80}
  n18 = LinearInterpolate(A=OneMinus(None=@n17), B=@n17, Alpha=NamedRerouteUsage())
  n19 = Multiply(A=NamedRerouteUsage(), B=@n18)
  n20 = ScalarParameter() {defaultValue=1, group=WeaveHarbor, parameterName=Height_01_Text, sortPriority=80}
  n21 = LinearInterpolate(A=OneMinus(None=@n20), B=@n20, Alpha=NamedRerouteUsage())
  n22 = Multiply(A=NamedRerouteUsage(), B=@n21)
  n23 = NamedRerouteUsage()
  n24 = LinearInterpolate(
   A=Min(A=Min(A=@n16, B=@n19), B=@n22)
   B=LinearInterpolate(A=LinearInterpolate(A=@n16, B=@n19, Alpha=@n23), B=@n22, Alpha=NamedRerouteUsage())
   Alpha=NamedRerouteUsage()
  )
  n25 = ScalarParameter() {defaultValue=1, group=WeaveHarbor, parameterName=Text_VirtualThickness}
  n26 = ScalarParameter() {defaultValue=2, group=WeaveHarbor, parameterName=Cloud_VirtualDistance}
  n27 = Add(A=Add(A=Multiply(A=@n24, B=@n25) {constB=5}), B=@n26)
  n28 = Multiply(A=DotProduct(A=NamedRerouteUsage(), B=VertexNormalWS())) {constB=-1}
  n29 = Add(A=Divide(A=Subtract(A=@n28, B=@n26), B=ScalarParameter() {defaultValue=1, group=WeaveHarbor, parameterName=Cloud_VirtualDensity}), B=@n26)
  n30 = Multiply(A=Max(A=Max(A=@n12, B=@n13), B=NamedRerouteUsage()), B=Saturate(None=Subtract(A=@n27, B=@n29)))
  n31 = VertexNormalWS()
  n32 = ScalarParameter() {defaultValue=-10, parameterName=FadeWPODistance}
  n33 = TransformPosition(None=Constant3Vector()) {transformType=TRANSFORMPOSSOURCE_TranslatedWorld}
  n34 = WorldPosition() {worldPositionShaderOffset=WPT_CameraRelativeNoOffsets}
  n35 = Subtract(A=@n33, B=@n34.XYZ)
  n36 = ScalarParameter() {parameterName=FadeWPOScaledownAmt, sliderMax=1}
  n37 = Add(
   A=Multiply(A=Multiply(A=@n31, B=@n32), B=NamedRerouteUsage()) {desc=Offset}
   B=Multiply(A=Multiply(A=@n35, B=@n36), B=NamedRerouteUsage()) {desc=Scale}
  )
  n38 = VertexNormalWS()
  n39 = Multiply(A=NamedRerouteUsage(), B=ScalarParameter() {defaultValue=0.13, parameterName=Twist Amount})
  n40 = Add(A=@n33, B=Constant3Vector() {constant=<rgba 0 0 2 0>})
  n41 = Add(A=@n37, B=@n34.XYZ)
  n42 = NamedRerouteDeclaration(None=Add(A=@n37, B=RotateAboutAxis(NormalizedRotationAxis=@n38, RotationAngle=@n39, PivotPoint=@n40, Position=@n41) {desc=Twist})) {name=WPO Anim, nodeColor=<rgba 1 0.6094 0 1>}
  n43 = DotProduct(
   A=WorldPosition() {worldPositionShaderOffset=WPT_CameraRelativeNoOffsets}.XYZ
   B=Transform(None=Constant3Vector() {constant=<rgba 0 1 0 0>}) {transformSourceType=TRANSFORMSOURCE_Local}
  )
  n44 = MaterialFunctionCall(Offset Amount=Multiply(A=ScalarParameter() {defaultValue=10, parameterName=CameraOffsetWPO}, B=Sign(None=@n43)) {desc=Invert when viewed from behind}) {materialFunction=/Engine/Functions/Engine_MaterialFunctions02/WorldPositionOffset/CameraOffset}
  MaterialFunctionCall(
   Material=SetMaterialAttributes!(
    MaterialAttributes=SetMaterialAttributes!(MaterialAttributes=@n1.Result, Base Color=@n11, Opacity=@n30) {attributeSetTypes [#1, #2]}
    World Position Offset=Add(A=@n42, B=@n44.Clamped Camera Offset)
   ) {attributeSetTypes [#3]}
   M=@n1.M
   Pre Skinned Local Position=@n1.M
   Distance Blend=@n1.M
   Fresnel=@n1.M
  ) {materialFunction=&MF_CharacterParent_2023_PostFX}
  n45 = MaterialFunctionCall(Speed=Constant() {r=0.25}) {materialFunction=&DebugTimeSine}
  MaterialFunctionCall!(In=Saturate(None=InverseLinearInterpolate(Value=@n45.Result) {constA=0.1, constB=0.9})) {materialFunction=/Game/Effects/Fort_Effects/Materials/_Functions/MF_Smoothstep_01}
  n46 = Constant() {r=0.5}
  n47 = ScalarParameter() {parameterName=TransitionOutDuration}
  n48 = Time()
  n49 = Constant() {r=5}
  n50 = MaterialFunctionCall(
   TimestampA=StaticSwitchParameter(True=@n46, False=ScalarParameter() {parameterName=TransitionOutStart}) {parameterName=DEBUG Loop Animation}
   TransitionLength_In=@n47
   TimestampB=StaticSwitchParameter(True=Add(A=@n46, B=@n47), False=ScalarParameter() {parameterName=TransitionInStart}) {parameterName=DEBUG Loop Animation}
   TransitionLength_Out=ScalarParameter() {parameterName=TransitionInDuration}
   Time=StaticSwitchParameter(True=Fmod(A=@n48, B=@n49), False=Time()) {parameterName=DEBUG Loop Animation}
   Transition In Only?=StaticBool()
  ) {materialFunction=&MF_Timestamp_InOut_01}
  n51 = CurveAtlasRowParameter(CurveTime=@n50."Linear (A Only)") {
   atlas=&/CA_Backpack_WeaveHarbor
   curve=&/C_Backpack_WeaveHarbor_Out
   parameterName=OutTransitionCurve
  }
  n52 = CurveAtlasRowParameter(CurveTime=@n50."Linear (B Only)") {
   atlas=&CA_Backpack_WeaveHarbor
   curve=&/C_Backpack_WeaveHarbor_In
   defaultValue=1
   parameterName=InTransitionCurve
  }
  NamedRerouteDeclaration(None=Multiply(A=OneMinus(None=@n51.R), B=OneMinus(None=@n52.R))) {name=Anim, nodeColor=<rgba 0.7813 0 1 1>}
  NamedRerouteDeclaration(None=Saturate(None=TwoSidedSign())) {name=IsFront, nodeColor=<rgba 0.1953 0 1 1>}
  NamedRerouteDeclaration(None=ScalarParameter() {group=WeaveHarbor, parameterName=HealthPhase}) {name=HealthPhase, nodeColor=<rgba 0.2656 1 0 1>}
  n53 = MaterialFunctionCall(UVs=TextureCoordinate(), Texture Scale=ScalarParameter() {defaultValue=0.8, parameterName=UV Scale}) {materialFunction=&ScaleUVsByCenter}
  n54 = ScalarParameter() {defaultValue=0.25, group=WeaveHarbor, parameterName=HeighRatio, sliderMax=1, sortPriority=80}
  n55 = BumpOffset(Coordinate=@n53.UVs, Height=@n21, HeightRatioInput=@n54) {heightRatio=0.25}
  n56 = NamedRerouteUsage()
  NamedRerouteDeclaration(None=Multiply(
    A=MaterialFunctionCall!(MipBias, UV=@n55, Phase=@n56) {materialFunction=&/MF_Backpack_WeaveHarbor_TextureSwap}.G
    B=MaterialFunctionCall(UVs=@n55) {materialFunction=&ScaleUVsByCenter}.0-1 mask
   )) {name=Mask01 Text, nodeColor=<rgba 0.4297 1 0 1>}
  n57 = BumpOffset(Coordinate=@n53.UVs, Height=@n18, HeightRatioInput=@n54) {heightRatio=0.25}
  n58 = NamedRerouteUsage()
  NamedRerouteDeclaration(None=Multiply(
    A=MaterialFunctionCall(UV=@n57, Phase=@n58) {materialFunction=&MF_Backpack_WeaveHarbor_TextureSwap}.B
    B=MaterialFunctionCall(UVs=@n57) {materialFunction=&ScaleUVsByCenter}.0-1 mask
   )) {name=Mask02 Shdw, nodeColor=<rgba 1 0 0.75 1>}
  n59 = BumpOffset(Coordinate=@n53.UVs, Height=@n15, HeightRatioInput=@n54) {heightRatio=0.25}
  n60 = NamedRerouteUsage()
  n61 = NamedRerouteDeclaration(None=Multiply(
    A=MaterialFunctionCall(UV=@n59, Phase=@n60) {materialFunction=&MF_Backpack_WeaveHarbor_TextureSwap}.R
    B=MaterialFunctionCall(UVs=@n59) {materialFunction=&ScaleUVsByCenter}.0-1 mask
   )) {name=Mask03 Bgnd, nodeColor=<rgba 0 1 0.9297 1>}
  NamedRerouteDeclaration(None=Saturate(None=Multiply(A=@n61) {constB=2})) {name=Mask03 Bgnd Clip, nodeColor=<rgba 0 1 0.9531 1>}
  n62 = Add(A=Subtract(A=@n51.G) {constB=0.5}, B=Subtract(A=@n52.G) {constB=0.5})
  n63 = Abs(None=@n62)
  NamedRerouteDeclaration(None=Multiply(
    A=MaterialFunctionCall(In=@n63) {materialFunction=&MF_Smoothstep_01}.Result
    B=Sign(None=@n62)
   )) {name=TwistAnim, nodeColor=<rgba 0 0.4375 1 1>}
  ScalarParameter() {group=WeaveHarbor, parameterName=AnimPhase, sliderMax=1}
 ]
}
