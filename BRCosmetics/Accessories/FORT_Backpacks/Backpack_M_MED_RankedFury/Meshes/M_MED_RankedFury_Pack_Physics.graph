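# Physics asset for the Ranked Fury backpack mesh, defining bodies and joints for its crown and trapezoid pieces.
# Each of the five bones has a single capsule: the pack, two base anchors and two dyn bones.
# The dyn crown and trapezoid bones hang off their base bones with tight angular limits, the trapezoid tighter, allowing only slight wobble.
# Both base bones connect to the pack with free swing and twist.
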PhysicsAsset /BRCosmetics/Accessories/FORT_Backpacks/Backpack_M_MED_RankedFury/Meshes/M_MED_RankedFury_Pack_Physics
bodies {
 base_crown_pivot [{shapeType=Capsule, length=0.505, radius=0.505, shapeName=base_crown_pivot_capsule, rotation=<rot 0 -90 0>}]
 base_trapezoid [{shapeType=Capsule, length=1, radius=1, shapeName=base_trapezoid_capsule, rotation=<rot -90 0 0>}]
 dyn_crown_pivot [{shapeType=Capsule, length=2, radius=7, shapeName=dyn_crown_pivot_capsule, center=<4 0 0>, rotation=<rot -90 0 0>}]
 dyn_trapezoid [{shapeType=Capsule, length=2.5, radius=1, shapeName=dyn_trapezoid_capsule, center=<2 0 0>, rotation=<rot 90 0 -0.0002>}]
 pack [{shapeType=Capsule, length=0.505, radius=0.505, shapeName=pack_capsule, rotation=<rot 0 -90 0>}]
}
constraints [
 {bone1Name=dyn_crown_pivot, bone2Name=base_crown_pivot, swing1LimitDegrees=10, swing2LimitDegrees=10, twistLimitDegrees=10}
 {bone1Name=dyn_trapezoid, bone2Name=base_trapezoid, swing1LimitDegrees=3.5, swing2LimitDegrees=3.5, twistLimitDegrees=1}
 {
  bone1Name=base_crown_pivot
  bone2Name=pack
  swing1LimitDegrees=45
  swing1Motion=Free
  swing2LimitDegrees=45
  swing2Motion=Free
  twistLimitDegrees=45
  twistMotion=Free
 }
 {
  bone1Name=base_trapezoid
  bone2Name=pack
  swing1LimitDegrees=45
  swing1Motion=Free
  swing2LimitDegrees=45
  swing2Motion=Free
  twistLimitDegrees=45
  twistMotion=Free
 }
]
