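# Animation Blueprint for the WeaveHarbor backpack cosmetic that gives its TextPlain piece springy, floating motion.
# The pack and MSTR_TextPlain bones tilt with the wearer's spine bend, then feed a cached pose into spherically limited Anim Dynamics on dyn_TextPlain, plus sine-driven bobbing.
# Crouch, sprint and jump/fall states retune the dynamics alpha and motion scales each update.
# A RigidBody node referencing RankedFury's physics asset and several computed values never reach the output.
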AnimBlueprint /BRCosmetics/Accessories/FORT_Backpacks/Backpack_M_MED_WeaveHarbor/Meshes/M_MED_WeaveHarbor_Pack_AnimBP
target_skeleton=&/M_MED_WeaveHarbor_Pack_Skeleton
graphs [
 AnimGraph {nodes [
   Anim Dynamics - Bone: dyn_TextPlain!(
    ComponentPose
    Alpha=Get Alpha_anim_dynamic() -> (Alpha_anim_dynamic:double)
    GravityOverride:Vector="0,0,20"
    ComponentLinearAccScale:Vector=Get Linear_Acc_Scale() -> (Linear_Acc_Scale:Vector)
    ComponentLinearVelScale:Vector=Get Linear_Vel_Scale() -> (Linear_Vel_Scale:Vector)
    ExternalForce="0,0,0"
   ) -> (Pose) {
    alpha=0
    angularBiasOverride=1
    angularDampingOverride=0.75
    angularSpringConstant=20
    bGravityOverrideInSimSpace=true
    bLinearSpring=true
    bOverrideLinearDamping=true
    bUseGravityOverride=true
    bUsePlanarLimit=false
    bUseSphericalLimits=true
    gravityScale=0
    linearDampingOverride=0.999
    linearSpringConstant=22
    simulationSpace=BoneRelative
    alphaBoolBlend {blendOption=Linear}
    alphaScaleBiasClamp {bInterpResult=true}
    boundBone {boneName=dyn_TextPlain}
    componentAppliedLinearAccClamp=<65 55 50>
    componentLinearAccScale=<0.7 0.7 1.5>
    componentLinearVelScale=<0.7 0.7 1.5>
    gravityOverride {z=20}
    physicsBodyDefinitions [{
      sphereCollisionRadius=0
      boundBone {boneName=dyn_TextPlain}
      boxExtents=<20 8 20>
      constraintSetup {
       linearXLimitType=Free
       linearZLimitType=Free
       angularLimitsMax=<0 0 0>
       angularLimitsMin=<0 0 0>
       angularTarget=<0 0 0>
       linearAxesMax=<10 0 0>
       linearAxesMin=<-10 -10 0>
      }
      localJointOffset=<0 0 0>
     }]
    relativeSpaceBone {boneName=root}
    sphericalLimits [{limitRadius=7, limitType=Inner, drivingBone {boneName=dyn_TextPlain}, sphereLocalOffset=<0 0 0>}]
   }
   "Transform (Modify) Bone - Bone: dyn_TextPlain"(Alpha=1, Translation="0,0,0", Rotation="0,0,0", Scale="1,1,1") {translationMode=BMM_Additive, translationSpace=BCS_BoneSpace, alphaBoolBlend {blendOption=Linear}, boneToModify {boneName=dyn_TextPlain}}
   RigidBody!(ComponentPose, Alpha, OverrideWorldGravity:Vector=Get GravityOverride(), ExternalForce) -> (Pose) {
    bClampLinearTranslationLimitToRefPose=true
    bOverrideWorldGravity=true
    overridePhysicsAsset=/BRCosmetics/Accessories/FORT_Backpacks/Backpack_M_MED_RankedFury/Meshes/M_MED_RankedFury_Pack_Physics
    simulationSpace=BaseBoneSpace
    alphaBoolBlend {blendOption=Linear}
    baseBoneRef {boneName=root}
    componentAppliedLinearAccClamp=<3000 3000 3000>
    componentLinearAccScale=<1 1 1>
    componentLinearVelScale=<0.3 0.2 0.2>
    simSpaceSettings {
     maxAngularAcceleration=1000
     maxAngularVelocity=3
     maxLinearAcceleration=700
     maxLinearVelocity=500
     velocityScaleZ=0
     worldAlpha=0.35
     externalLinearDragV=<0.2 0.2>
    }
   }
   n1 = Anim Dynamics - Bone: dyn_TextPlain(
    ComponentPose=Local To Component(LocalPose=Use cached pose 'translation_Cache'())
    Alpha=Get Alpha_anim_dynamic()
    GravityOverride="0,0,5"
    ComponentLinearAccScale=Get Linear_Acc_Scale()
    ComponentLinearVelScale=Get Linear_Vel_Scale()
    ExternalForce="0,0,0"
   ) {
    alpha=0
    angularBiasOverride=1
    angularDampingOverride=0.75
    angularSpringConstant=20
    bGravityOverrideInSimSpace=true
    bLinearSpring=true
    bOverrideLinearDamping=true
    bUseGravityOverride=true
    bUsePlanarLimit=false
    bUseSphericalLimits=true
    gravityScale=0
    linearDampingOverride=0.98
    linearSpringConstant=22
    simulationSpace=BoneRelative
    alphaBoolBlend {blendOption=Linear}
    alphaScaleBiasClamp {bInterpResult=true}
    boundBone {boneName=dyn_TextPlain}
    componentAppliedLinearAccClamp=<300 1 300>
    componentLinearAccScale=<0.7 0.7 1.5>
    componentLinearVelScale=<0.7 0.7 1.5>
    gravityOverride {z=20}
    physicsBodyDefinitions [{
      sphereCollisionRadius=0
      boundBone {boneName=dyn_TextPlain}
      boxExtents=<17 17 17>
      constraintSetup {
       linearXLimitType=Free
       linearZLimitType=Free
       angularLimitsMax=<0 0 0>
       angularLimitsMin=<0 0 0>
       angularTarget=<0 0 0>
       linearAxesMax=<10 0 10>
       linearAxesMin=<-10 0 -10>
      }
      localJointOffset=<1 0 0>
     }]
    relativeSpaceBone {boneName=root}
    sphericalLimits [{limitRadius=5, limitType=Inner, drivingBone {boneName=MSTR_TextPlain}, sphereLocalOffset=<0 0 0>}]
   }
   n2 = "Transform (Modify) Bone - Bone: dyn_TextPlain"(ComponentPose=@n1, Alpha=1, Translation=Get Translate Main() -> (Translate Main:Vector), Rotation="0,0,0", Scale="1,1,1") {translationMode=BMM_Additive, translationSpace=BCS_BoneSpace, alphaBoolBlend {blendOption=Linear}, boneToModify {boneName=dyn_TextPlain}}
   Output Pose(Result=Component To Local(ComponentPose=@n2)) {layerGroup=None}
   n3 = "Transform (Modify) Bone - Bone: pack"(ComponentPose=Local To Component(LocalPose=Input Pose()), Translation="0,-0.5,3", Rotation="0,0,0") {
    translationMode=BMM_Additive
    translationSpace=BCS_BoneSpace
    alphaBoolBlend {blendOption=Linear}
    boneToModify {boneName=pack}
    translation {y=-0.5, z=3}
   }
   n4 = "Transform (Modify) Bone - Bone: pack"(ComponentPose=@n3, Translation="0,0,0", Rotation=Get Backpack Rotate() -> (Backpack Rotate:Rotator)) {
    rotationMode=BMM_Additive
    rotationSpace=BCS_BoneSpace
    translationMode=BMM_Additive
    translationSpace=BCS_BoneSpace
    alphaBoolBlend {blendOption=Linear}
    boneToModify {boneName=pack}
   }
   n5 = "Transform (Modify) Bone - Bone: MSTR_TextPlain"(ComponentPose=@n4, Alpha=1, Translation="0,0,0", Rotation=Get Backpack Rotate(), Scale="1,1,1") {
    rotationMode=BMM_Additive
    rotationSpace=BCS_BoneSpace
    translationMode=BMM_Additive
    translationSpace=BCS_BoneSpace
    alphaBoolBlend {blendOption=Linear}
    boneToModify {boneName=MSTR_TextPlain}
   }
   Save cached pose 'translation_Cache'(Pose=Component To Local(ComponentPose=@n5)) {cacheName=translation_Cache}
   Get GravityOverride()
   Get Backpack Translation() -> (Backpack Translation:Vector)
   Get Backpack Translation()
  ]}
 EventGraph {nodes [
   n1 = Event BlueprintUpdateAnimation()
   > n2 = Sequence!(execute) -> (then_0, then_1, then_2:Exec, then_3:Exec)
   n3 = float * float(A=GetGameTimeInSeconds(), B=2) {is=title K2Node_PromotableOperator}
   n4 = "Sin (Radians)"(A=float * float(A=0.6, B=@n3) {is=title K2Node_CommutativeAssociativeBinaryOperator})
   n5 = "Sin (Radians)"(A=float * float(A=0.8, B=@n3))
   Set Translate Main(execute=@n2.then_2, Translate Main:Vector=MakeVector(X=float * float(A=@n4, B=2), Y=0, Z=float * float(A=@n5, B=2))) -> (Output_Get:Vector)
   n6 = Gravity Override State Function DEPRECATED(
    execute=@n2.then_0
    CharacterPartAnimInstance=Self-Reference()
    GameTimeInSeconds=GetGameTimeInSeconds()
    GravityOverrideParams=Get Gravity Override Params()
   )
   > Set GravityOverride(GravityOverride=@n6)
   > n7 = Branch(Condition=Get bIsSkyDiving())
   > Set RB Alpha(RB Alpha:double=0.7) -> (Output_Get:double)
   Set RB Alpha(execute=@n7.else, RB Alpha=1)
   n8 = Branch(execute=@n2.then_3, Condition=Get bIsCrouching())
   > Set Alpha_anim_dynamic(Alpha_anim_dynamic:double=0.3) -> (Output_Get:double)
   > Set Linear_Acc_Scale(Linear_Acc_Scale:Vector="0,0,0.1") -> (Output_Get:Vector)
   > n9 = Set Linear_Vel_Scale(Linear_Vel_Scale:Vector="1.2,1.2,0.6") -> (Output_Get:Vector)
   n10 = Get bIsCrouching()
   n11 = Branch(execute=[@n15.then, @n9.then], Condition=AND Boolean(A=Get bIsSprinting(), B=NOT Boolean(A=@n10)))
   > Set Alpha_anim_dynamic(Alpha_anim_dynamic=0.1)
   > Set Linear_Acc_Scale(Linear_Acc_Scale="0,0,0.1")
   > n12 = Set Linear_Vel_Scale(Linear_Vel_Scale="1,1,1.2")
   n13 = Branch(execute=[@n14.then, @n12.then], Condition=OR Boolean(A=Get bIsJumping(), B=Get bIsFalling()))
   > Set Alpha_anim_dynamic(Alpha_anim_dynamic=0.1)
   > Set Linear_Acc_Scale(Linear_Acc_Scale="0, 0, 0")
   > Set Linear_Vel_Scale(Linear_Vel_Scale="0.2,0.2,0.8")
   Set Alpha_anim_dynamic(execute=@n13.else, Alpha_anim_dynamic=1)
   > Set Linear_Acc_Scale(Linear_Acc_Scale="1.5,1.1,1.5")
   > Set Linear_Vel_Scale(Linear_Vel_Scale="1.5,1.1,1.5")
   Set Alpha_anim_dynamic(execute=@n11.else, Alpha_anim_dynamic=1)
   > Set Linear_Vel_Scale(Linear_Vel_Scale="1.5,1.1,1.5")
   > n14 = Set Linear_Vel_Scale(Linear_Vel_Scale="1.5,1.1,1.5")
   Set Alpha_anim_dynamic(execute=@n8.else, Alpha_anim_dynamic=1)
   > Set Linear_Vel_Scale(Linear_Vel_Scale="1.5,1.1,1.5")
   > n15 = Set Linear_Vel_Scale(Linear_Vel_Scale="1.5,1.1,1.5")
   n16 = GetOwnerSkeletalMeshComponent()
   n17 = BreakRotator(InRot=BreakTransform(InTransform=GetDeltaTransformFromRefPose(self=@n16, BoneName=spine_05, BaseName=spine_01)).Rotation)
   Set Backpack Rotate(execute=@n2.then_1, Backpack Rotate:Rotator=MakeRotator(Roll=MapRangeClamped(Value=@n17.Yaw, InRangeA=0, InRangeB=-40, OutRangeB=30))) -> (Output_Get:Rotator)
   > Set Backpack Translation(Backpack Translation:Vector=MakeVector(Y=MapRangeClamped(Value=@n17.Yaw, InRangeA=0, InRangeB=-30, OutRangeB=-5))) -> (Output_Get:Vector)
   n18 = GetOwnerSkeletalMeshComponent()
   > n19 = Set head down rot(head down rot:double=GetAlphaFromReferenceTransformDelta(SkeletalMeshComponent=@n18, BoneName=spine_01, BaseName=head, InputRange=(X=0,Y=-45), TwistAxis="1,0,0")) -> (Output_Get:double)
   n20 = "To Float (Boolean)"(InBool=Get bIsPlayingMeleeAnim())
   n21 = float * float(A=@n1.DeltaTimeX, B=0.2) {is=title K2Node_PromotableOperator}
   n22 = Get isMelee() -> (isMelee:double)
   > Set isMelee(isMelee:double=float * float(A=@n19, B=FInterpTo(Current=@n22, Target=@n20, DeltaTime=@n21, InterpSpeed=5))) -> (Output_Get:double)
   n23 = BreakRotator(InRot=BreakTransform(InTransform=GetDeltaTransformFromRefPose(self=@n18, BoneName=upperarm_l, BaseName=clavicle_l)).Rotation)
   > Set Backpack Rotate Pickaxe(Backpack Rotate Pickaxe:Rotator=MakeRotator(Roll=MapRangeClamped(Value=@n23.Yaw, InRangeA=0, InRangeB=90, OutRangeB=30))) -> (Output_Get:Rotator)
   Reroute Node()
   > n24 = Branch(Condition=Get bIsCrouching())
   > Set Alpha_anim_dynamic(Alpha_anim_dynamic=0.3)
   > Set Linear_Acc_Scale(Linear_Acc_Scale="0,0,0.1")
   > n25 = Set Linear_Vel_Scale(Linear_Vel_Scale="1.2,1.2,0.6")
   n26 = Get bIsCrouching()
   n27 = Branch(execute=[@n31.then, @n25.then], Condition=AND Boolean(A=Get bIsSprinting(), B=NOT Boolean(A=@n26)))
   > Set Alpha_anim_dynamic(Alpha_anim_dynamic=0.1)
   > Set Linear_Acc_Scale(Linear_Acc_Scale="0,0,0.1")
   > n28 = Set Linear_Vel_Scale(Linear_Vel_Scale="1,1,1.2")
   n29 = Branch(execute=[@n30.then, @n28.then], Condition=OR Boolean(A=Get bIsJumping(), B=Get bIsFalling()))
   > Set Alpha_anim_dynamic(Alpha_anim_dynamic=0.1)
   > Set Linear_Acc_Scale(Linear_Acc_Scale="0, 0, 0")
   > Set Linear_Vel_Scale(Linear_Vel_Scale="0.2,0.2,0.8")
   Set Alpha_anim_dynamic(execute=@n29.else, Alpha_anim_dynamic=0.7)
   > Set Linear_Acc_Scale(Linear_Acc_Scale="1.1,1.1,1.3")
   > Set Linear_Vel_Scale(Linear_Vel_Scale="1.1,1.1,1.3")
   Set Alpha_anim_dynamic(execute=@n27.else, Alpha_anim_dynamic=0.7)
   > Set Linear_Vel_Scale(Linear_Vel_Scale="1.1,1.1,1.3")
   > n30 = Set Linear_Vel_Scale(Linear_Vel_Scale="1.1,1.1,1.3")
   Set Alpha_anim_dynamic(execute=@n24.else, Alpha_anim_dynamic=0.7)
   > Set Linear_Vel_Scale(Linear_Vel_Scale="1.1,1.1,1.3")
   > n31 = Set Linear_Vel_Scale(Linear_Vel_Scale="1.1,1.1,1.3")
   n32 = GetRealTimeSeconds()
   n33 = float * float(A=float * float(A=1, B=-1.2), B=@n32)
   n34 = float * float(A=float * float(A=0.4, B=-0.8), B=@n32)
   MakeVector(X=float / float(A="Sin (Radians)"(A=@n33), B=-1.3), Y=0, Z=float / float(A="Sin (Radians)"(A=@n34), B=-1.2))
   TryGetPawnOwner()
   n35 = GetRealTimeSeconds()
   n36 = float * float(A=float * float(A=1, B=-1.6), B=@n35)
   n37 = float * float(A=float * float(A=0.4, B=-0.8), B=@n35)
   MakeVector(X=float / float(A="Sin (Radians)"(A=@n36), B=0.85), Y=0, Z=float / float(A="Sin (Radians)"(A=@n37), B=1.4))
   n38 = GetRealTimeSeconds()
   n39 = float * float(A=float * float(A=0.4, B=-0.8), B=@n38)
   n40 = float * float(A=float * float(A=1, B=-1.2), B=@n38)
   MakeVector(X=float / float(A="Sin (Radians)"(A=@n40), B=-10), Y=0, Z=float / float(A="Sin (Radians)"(A=@n39), B=-10))
   n41 = float * float(A=1, B=1.1)
   float / float(A="Sin (Radians)"(A=float * float(A=@n41, B=@n32)), B=20)
   n42 = float * float(A=0.7, B=1.3)
   float / float(A="Sin (Radians)"(A=float * float(A=@n42, B=@n35)), B=4)
   n43 = float * float(A=1, B=1.1)
   float / float(A="Sin (Radians)"(A=float * float(A=@n43, B=@n38)), B=20)
  ]}
]
variables [
 Alpha_anim_dynamic {type=number}
 Backpack Rotate {type=Rotator}
 Backpack Rotate Pickaxe {type=Rotator}
 Backpack Translation {type=Vector}
 Backpack_floating {type=Vector}
 Gravity Override Params {type=GravityOverrideParamsStruct}
 GravityOverride {type=Vector}
 head down rot {type=number}
 isMelee {type=number}
 Linear_Acc_Scale {type=Vector}
 Linear_Vel_Scale {type=Vector}
 RB Alpha {type=number}
 Translate Main {type=Vector}
]
